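# Niagara burst effect for the TealMink emote's bag poof, throwing out a cloud, coins and debris.
# Emitters include MainCloud puffs, speed-stretched Pointy streaks, tumbling IceChunks meshes, and Coins randomly split between mesh and sprite with selective collision and sine-driven flip and shimmer.
# Bursts ease out through drag and fading curves, while time-of-day and front-end switches adjust color, brightness and dynamic material parameters.
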NiagaraSystem /BRCosmetics/Animation/Game/MainPlayer/Emotes/TealMink/FX/NS_Emote_TealMink_BagPoof
dependencies {
 usedDataInterfaces [
  &NiagaraDataInterfaceCurve
  &NiagaraDataInterfaceColorCurve
  &NiagaraDataInterfaceArrayInt32
  &NiagaraDataInterfaceSpriteRendererInfo
  &NiagaraDataInterfaceMeshRendererInfo
 ]
 usedDynamicInputs []
 usedModules [
  &SystemState
  &self:SystemSpawnScript.SystemScriptSource.SystemScriptGraph.NiagaraNodeAssignment_0.SetVariables_#1
 ]
}
emitters {
 Coin {
  emitterUpdateScript [
   &EmitterState(
    Life Cycle Mode=::NewEnumerator1
    Loop Behavior=::NewEnumerator1
    Loop Delay=&RandomRangeFloat(Minimum=2, Maximum=4)
    Loop Duration Mode=::NewEnumerator1
    UseLoopDelay=false
   )
   &SpawnBurst_Instantaneous(Spawn Count=150, Use Spawn Probability=false, UseLoopCountLimit=false)
   /SetVariables(Emitter.SineWave:NiagaraFloat=&Sine(Normalized Angle=$Emitter.Age, Period=&RandomRangeFloat(Minimum=0.4, Maximum=1.1), Scale=1, Bias=0)) {isSetParameters=true}
  ]
  particleSpawnScript [
   /SetVariables(Particles.RendererChoice:NiagaraInt32=&RandomRangeInt(Minimum=0, Maximum=1)) {isSetParameters=true}
   /SetVariables(Particles.VisibilityTag=$Particles.RendererChoice) {isSetParameters=true}
   &InitializeParticle(
    Color=&MakeCustomLinearColorFromBool(
     Bool=&Boolean_OrOperation(Bool1=$User.bIsFrontEnd, Bool2=$User.bIsFrontEndPreview)
     True Color=<rgba 0.3311 1 0.6483 1>
     False Color=<rgba 1 0.9843 0.9961 1>
    )
    Lifetime Mode=::NewEnumerator1
    Mesh Scale Mode=::NewEnumerator3
    Mesh Uniform Scale=&RandomRangeFloat(Minimum=0.05, Maximum=0.03)
    Sprite Rotation Mode=::NewEnumerator2
    Sprite Size Mode=::NewEnumerator3
    Uniform Sprite Size=5
    UsePositionOffset=false
    Write Material Random=true
   )
   &ApplyInitialForces()
   &AddVelocityFromPoint(
    Module.Use Falloff Distance=false
    Origin Offset=&RandomRangeVector(Minimum=<0 0 0>, Maximum=<0 0 0>)
    Velocity Strength=&RandomRangeFloat(Minimum=99, Maximum=200)
   ) {enabled=false}
   &SphereLocation(
    MaskSpawnGroup=false
    Offset=<0 0 -10>
    Offset Coordinate Space:ENiagaraCoordinateSpace=::World
    Override Seed=true
    Random Seed=20
    Sphere Radius=5
    Surface Only=false
    UseNonUniformScale=false
    UseOffset=true
    UseSphereOrientation=false
   )
   &InitialMeshOrientation(Use Rotation Vector=false)
   /SetVariables(Particles.MeshOpacity:LinearColor=<rgba 0.7 0.3879 0.1268 0>, Particles.SpriteColor=<rgba 1 0.9843 0.9961 1>) {isSetParameters=true}
   &AddVelocity(
    Rotation Coordinate Space=::World
    Velocity=&RandomRangeVector(Minimum=<150 150 20>, Maximum=<-150 -150 600>)
   )
  ]
  particleUpdateScript [
   /SetVariables(Particles.SineWave:NiagaraFloat=&Sine(Normalized Angle=$Emitter.Age, Period=&RandomRangeFloat(Minimum=0.3, Maximum=0.7), Scale=1, Bias=0)) {isSetParameters=true}
   &ParticleState()
   &ScaleSpriteSize(Scale Factor=&MakeVector2D(X=$Particles.SineWave, Y=1), Scale Sprite Size Mode=::NewEnumerator2) {module=ScaleSpriteSize001}
   &SpriteRotationRate(Rotation Rate=&RandomRangeFloat(Minimum=-1000, Maximum=-500))
   &ScaleMeshSize(Scale Factor=&VectorFromFloat(Value=&FloatFromCurve(FloatCurve=DataInterface() {propertyValues {Curve {keys [<at 0.6 1>,<at 1 0>]}}})))
   &ScaleSpriteSize(Uniform Curve Sprite Scale=DataInterface() {propertyValues {Curve {keys [<at 0.001 0.5>,<at 0.2948 1.1652>,<at 1.0009 0.8072>]}}})
   &ScaleColor(
    Linear Color Curve=DataInterface() {propertyValues {
      AlphaCurve {keys [<at 0.0057 1 RCIM_Cubic>]}
      BlueCurve {keys [<at 0 0.6519>,<at 0.666 0.7333>]}
      GreenCurve {keys [<at 0 0.414>,<at 0.666 0.6>]}
      RedCurve {keys [<at 0 2>,<at 0.666 1>]}
     }}
    Scale Mode=::NewEnumerator2
   ) {enabled=false, module=ScaleColor001}
   &ScaleColor(
    Scale Alpha=&FloatFromCurve(
     FloatCurve=DataInterface() {propertyValues {Curve {keys [
         {arriveTangent=-0.0888, interpMode=RCIM_Cubic, leaveTangent=-0.0888, tangentMode=RCTM_User, time=0, value=0}
         {arriveTangent=0.12, interpMode=RCIM_Cubic, leaveTangent=0.12, tangentMode=RCTM_User, time=0.15, value=0.9847}
         {arriveTangent=-0.0405, interpMode=RCIM_Cubic, leaveTangent=-0.0405, tangentMode=RCTM_User, time=1, value=0}
        ]}}}
     Scale Curve=&FNTimeOfDayScalar(
      Night Scale=1
      Morning Scale=1
      Day Scale=&FN_FrontEndFloatModifier(Gameplay=1, Front End=1)
      Evening Scale=1
     )
    )
    Scale RGB=&VectorFromFloat(Value=&FNTimeOfDayScalar(
      Night Scale=0.8
      Morning Scale=0.8
      Day Scale=&FN_FrontEndFloatModifier(Gameplay=1, Front End=1)
      Evening Scale=0.5
     ))
    ScaleA=true
    ScaleRGB=true
   ) {module=ScaleColor003}
   &Drag(
    Drag=&FloatFromCurve(
     FloatCurve=DataInterface() {propertyValues {Curve {keys [
         {arriveTangent=0.7001, interpMode=RCIM_Cubic, leaveTangent=0.7001, tangentMode=RCTM_User, time=0, value=0}
         {interpMode=RCIM_Cubic, tangentMode=RCTM_User, time=1, value=1}
        ]}}}
     Scale Curve=4
    )
    Use Linear Drag=true
    Use Rotational Drag=true
   ) {module=Drag002}
   &GravityForce()
   &UpdateMeshOrientation(Rotation Vector=&Multiply_Vector(
     A=<1 1 1>
     B=&MakeVector(
      X=&MakeCustomFloatFromBool(Bool=&RandomBool(), True Float=2, False Float=-2)
      Y=&RandomRangeFloat(Minimum=0.1, Maximum=-0.1)
      Z=&RandomRangeFloat(Minimum=0.1, Maximum=-0.1)
     )
    ))
   &Collision(
    Age Colliding Particles=false
    Collision Enabled=&SetBoolByFloatComparison(
     A=&RandomRangeFloat(
      Minimum=0
      Maximum=&Multiply_FloatByInt(
       Float=1
       Integer=&SelectIntFromArray(Direct Array Index=$Particles.RendererChoice, Int Selection Array=DataInterface() {propertyValues {IntData [1,0]}})
      )
     )
     B=0.8
    )
    Control Roll On Collision=true
    KillOnCollision=false
    Module.Control Rest State Via Surface Normal=false
    Module.EnableMaxCollisionCount=false
    Module.Randomize Collision Normal Vector=false
    Restitution=1
    ScaleRateOfRotation=false
   )
   &SolveForcesAndVelocity(Clamp Velocity=false, Limit Acceleration=false)
   &DynamicMaterialParameters(
    Index 0 Param 1=&Add_Float(
     A=&RandomRangeFloat(Minimum=0, Maximum=1)
     B=&FloatFromCurve(
      FloatCurve=DataInterface() {propertyValues {Curve {keys [
          {arriveTangent=2.7533, interpMode=RCIM_Cubic, leaveTangent=1.8958, tangentMode=RCTM_User, time=0, value=1}
          {arriveTangent=0.8292, interpMode=RCIM_Cubic, leaveTangent=0.8292, tangentMode=RCTM_User, time=0.321, value=1.3453}
          {arriveTangent=0.0529, interpMode=RCIM_Cubic, leaveTangent=0.2021, tangentMode=RCTM_User, time=1, value=1.5}
         ]}}}
      CurveIndex=&Add_Float(A=&Multiply_Float(A=1, B=0.5), B=$Particles.NormalizedAge)
     )
    )
    Index 0 Param 2=&FloatFromCurve(FloatCurve=DataInterface() {propertyValues {Curve {keys [
         {arriveTangent=0.0453, interpMode=RCIM_Cubic, tangentMode=RCTM_Break, time=0, value=-1}
         {arriveTangent=0.0604, interpMode=RCIM_Cubic, leaveTangent=0.0604, tangentMode=RCTM_User, time=1, value=-1}
        ]}}})
    Index 0 Param 3=&FloatFromCurve(FloatCurve=DataInterface() {propertyValues {Curve {keys [
         {arriveTangent=-53.9199, interpMode=RCIM_Cubic, leaveTangent=-8.9727, tangentMode=RCTM_User, time=0, value=1}
         {arriveTangent=0.2347, interpMode=RCIM_Cubic, leaveTangent=0.2723, tangentMode=RCTM_Break, time=0.25, value=0}
        ]}}})
    Index 0 Param 4=&FloatFromCurve(
     FloatCurve=DataInterface() {propertyValues {Curve {keys [<at 0.5 1>,<at 1 0>]}}}
     Scale Curve=&FNTimeOfDayScalar(
      Night Scale=2
      Morning Scale=2
      Day Scale=&FN_FrontEndFloatModifier(Gameplay=1, Front End=0.5)
      Evening Scale=3
     )
    )
    Index 2 Param 1=&FloatFromCurve(FloatCurve=DataInterface() {propertyValues {Curve {keys [
         {interpMode=RCIM_Cubic, leaveTangent=4.2485, tangentMode=RCTM_Break, time=0, value=0}
         {arriveTangent=1.0548, interpMode=RCIM_Cubic, leaveTangent=1.0548, tangentMode=RCTM_User, time=0.17, value=0.5715}
         <at 1 1>
        ]}}})
    Index 2 Param 2=&FloatFromCurve(
     FloatCurve=DataInterface() {propertyValues {Curve {keys [
         {arriveTangent=1.9056, interpMode=RCIM_Cubic, leaveTangent=1.9056, tangentMode=RCTM_User, time=0, value=0.2}
         {arriveTangent=0.0259, interpMode=RCIM_Cubic, leaveTangent=0.0259, tangentMode=RCTM_User, time=1, value=1.4}
        ]}}}
     Scale Curve=&RandomRangeFloat(Minimum=0.2, Maximum=0.5)
    )
    Index 2 Param 3=&FloatFromCurve(FloatCurve=DataInterface() {propertyValues {Curve {keys [
         {interpMode=RCIM_Cubic, leaveTangent=2.3064, tangentMode=RCTM_Break, time=0, value=1}
         {interpMode=RCIM_Cubic, leaveTangent=11.1233, tangentMode=RCTM_Break, time=1, value=2}
        ]}}})
    Index 2 Param 4=&FloatFromCurve(
     FloatCurve=DataInterface() {propertyValues {Curve {keys [
         {arriveTangent=-0.6867, interpMode=RCIM_Cubic, leaveTangent=-0.6867, tangentMode=RCTM_User, time=0, value=1}
         {interpMode=RCIM_Cubic, leaveTangent=-1.2956, tangentMode=RCTM_Break, time=1, value=0.75}
        ]}}}
     Scale Curve=1.6
    )
    Module.Param0WriteEnabled=true
    Module.Param0WriteEnabled2=true
    Module.Param1WriteEnabled=false
    Module.Param1WriteEnabled2=true
    Module.Param2WriteEnabled=false
    Module.Param2WriteEnabled2=true
    Module.Param3WriteEnabled=true
    Module.Param3WriteEnabled2=true
    Write Parameter Index 2=true
   )
   &SubUVAnimation(
    Evaluation Type=::NewEnumerator0
    "Sprite Renderer "=DataInterface() {propertyValues {SpriteRenderer=&self:Coin_0.NiagaraSpriteRendererProperties_0}}
    SubUV Animation Mode:ENiagara_SubUVLookupModeV2=::NewEnumerator2
    UseEndFrame=false
    UseStartFrame=false
   )
   &ScaleColor(
    Scale RGB=&VectorFromFloat(Value=&RemapRange(Input Value To Remap=$Particles.SineWave, Input Min=-1, Input Max=1, Output Min=1, Output Max=3))
    ScaleA=true
    ScaleRGB=true
   )
  ]
  rendererClasses [&NiagaraMeshRendererProperties, &NiagaraSpriteRendererProperties]
  renderers [
   {rendererClass=&NiagaraMeshRendererProperties, rendererIndex=0}
   {rendererClass=&NiagaraSpriteRendererProperties, rendererIndex=1}
  ]
 }
 IceChunks {
  emitterUpdateScript [
   &EmitterState(Life Cycle Mode=::NewEnumerator1, Loop Behavior=::NewEnumerator1, Loop Duration=2, UseLoopDelay=false)
   &SpawnBurst_Instantaneous(Spawn Count=180, Use Spawn Probability=false, UseLoopCountLimit=false)
  ]
  particleSpawnScript [
   &InitializeParticle(
    Color=<rgba 7.5 1.77 3.1452 1>
    Lifetime Max=1
    Lifetime Min=0.25
    Lifetime Mode=::NewEnumerator1
    Mesh Renderer Array Visibility Mode=::NewEnumerator1
    Mesh Renderer Info=DataInterface() {propertyValues {MeshRenderer=&self:IceChunks_0.NiagaraMeshRendererProperties_0}}
    Mesh Renderer Mesh Index:NiagaraInt32=&RandomRangeInt(Minimum=0, Maximum=2)
    Mesh Scale Mode=::NewEnumerator3
    Mesh Uniform Scale=&RandomRangeFloat(Minimum=0.1, Maximum=0.15)
    UsePositionOffset=false
    Write Material Random=true
   )
   &ShapeLocation(MaskSpawnGroup=false, Override Seed=false, Sphere Radius=11, UseOwnerScale=true)
   &InitialMeshOrientation(
    Mesh Axis To Orient=<0 0 1>
    Mesh Orientation Mode=::NewEnumerator2
    Orientation Vector=$Particles.ShapeLocation.ShapeNormal
    Use Rotation Vector=false
   )
   &AddVelocity(Module.Use Falloff Distance=false, Override Seed=false, Velocity Mode=::NewEnumerator1, Velocity Speed=400)
   &AddVelocity(Rotation Coordinate Space=::World, Velocity=&RandomRangeVector(Minimum=<0 0 50>, Maximum=<0 0 400>)) {module=AddVelocity001}
  ]
  particleUpdateScript [
   &ParticleState()
   &UpdateMeshOrientation(
    Rotation Rate=&RandomRangeFloat(Minimum=2, Maximum=4)
    Rotation Vector=&RandomVector(Vector Scale=1)
   )
   &ScaleMeshSize(Scale Factor=&VectorFromFloat(Value=&FloatFromCurve(FloatCurve=DataInterface() {propertyValues {Curve {keys [
          {interpMode=RCIM_Cubic, tangentMode=RCTM_User, time=0, value=1}
          {arriveTangent=-1.6697, interpMode=RCIM_Cubic, leaveTangent=-1.6697, tangentMode=RCTM_User, time=1, value=0}
         ]}}})))
   &Drag(Drag=&RandomRangeFloat(Minimum=2, Maximum=3), Use Linear Drag=true, Use Rotational Drag=true)
   &CurlNoiseForce(Noise Strength=111, Pan Noise=true, Pan Noise Field=<0 0 0.1>)
   &GravityForce()
   &SolveForcesAndVelocity(Clamp Velocity=false, Limit Acceleration=false)
  ]
  rendererClasses [&NiagaraMeshRendererProperties]
  renderers [{rendererClass=&NiagaraMeshRendererProperties, rendererIndex=0}]
 }
 MainCloud {
  emitterUpdateScript [
   &EmitterState(
    Life Cycle Mode=::NewEnumerator1
    Loop Behavior=::NewEnumerator1
    Loop Delay=&RandomRangeFloat(Minimum=2, Maximum=4)
    Loop Duration Mode=::NewEnumerator1
    UseLoopDelay=false
   )
   &SpawnBurst_Instantaneous(Spawn Count=25, Use Spawn Probability=false, UseLoopCountLimit=false)
  ]
  particleSpawnScript [
   &InitializeParticle(
    Color=&FNTimeOfDayColor(
     Night Color=<rgba 1 0.4314 0.5176 0.502>
     Morning Color=<rgba 0.4378 0.5174 1 1>
     Day Color=<rgba 1.1027 0.5044 1.5 0.5>
     Evening Color=<rgba 1 0.4314 0.5176 0.502>
    )
    Lifetime Max=2.3
    Lifetime Mode=::NewEnumerator1
    Sprite Size Mode=::NewEnumerator1
    Uniform Sprite Size Max=140
    Uniform Sprite Size Min=80
    UsePositionOffset=false
    Write Material Random=true
   )
   &ApplyInitialForces()
   &AddVelocityFromPoint(
    Module.Use Falloff Distance=false
    Origin Offset=&RandomRangeVector(Minimum=<0 0 0>, Maximum=<0 0 0>)
    Velocity Strength=&RandomRangeFloat(Minimum=99, Maximum=150)
   )
   &SphereLocation(
    MaskSpawnGroup=false
    Offset=<0 0 -20>
    Offset Coordinate Space:ENiagaraCoordinateSpace=::World
    Override Seed=true
    Random Seed=20
    Sphere Radius=5
    Surface Only=false
    UseNonUniformScale=false
    UseOffset=true
    UseSphereOrientation=false
   )
   /SetVariables(Particles.Unlit Color:LinearColor=<rgba 1 0.851 0.9591 1>) {isSetParameters=true, module=SetVariables_#2}
  ]
  particleUpdateScript [
   &ParticleState()
   &ScaleSpriteSize(Uniform Curve Sprite Scale=DataInterface() {propertyValues {Curve {keys [<at 0.001 0.5>,<at 0.1 1.1652>,<at 1.0009 0.8072>]}}})
   &ScaleColor(
    Linear Color Curve=DataInterface() {propertyValues {
      AlphaCurve {keys [<at 0.0057 1 RCIM_Cubic>]}
      BlueCurve {keys [<at 0 0.6519>,<at 0.666 0.7333>]}
      GreenCurve {keys [<at 0 0.414>,<at 0.666 0.6>]}
      RedCurve {keys [<at 0 2>,<at 0.666 1>]}
     }}
    Scale Mode=::NewEnumerator2
   ) {enabled=false, module=ScaleColor001}
   &ScaleColor(
    Scale Alpha=&FloatFromCurve(
     FloatCurve=DataInterface() {propertyValues {Curve {keys [
         {arriveTangent=-0.0888, interpMode=RCIM_Cubic, leaveTangent=-0.0888, tangentMode=RCTM_User, time=0, value=0}
         {arriveTangent=0.12, interpMode=RCIM_Cubic, leaveTangent=0.12, tangentMode=RCTM_User, time=0.15, value=0.9847}
         {arriveTangent=-0.0405, interpMode=RCIM_Cubic, leaveTangent=-0.0405, tangentMode=RCTM_User, time=1, value=0}
        ]}}}
     Scale Curve=&FNTimeOfDayScalar(
      Night Scale=1
      Morning Scale=1
      Day Scale=&FN_FrontEndFloatModifier(Gameplay=1, Front End=1)
      Evening Scale=1
     )
    )
    Scale RGB=&VectorFromFloat(Value=&FNTimeOfDayScalar(
      Night Scale=0.8
      Morning Scale=0.8
      Day Scale=&FN_FrontEndFloatModifier(Gameplay=1, Front End=1)
      Evening Scale=0.5
     ))
    ScaleA=true
    ScaleRGB=true
   ) {module=ScaleColor003}
   &Drag(
    Drag=&FloatFromCurve(
     FloatCurve=DataInterface() {propertyValues {Curve {keys [
         {arriveTangent=0.7001, interpMode=RCIM_Cubic, leaveTangent=0.7001, tangentMode=RCTM_User, time=0, value=0}
         {interpMode=RCIM_Cubic, tangentMode=RCTM_User, time=1, value=1}
        ]}}}
     Scale Curve=4
    )
    Use Linear Drag=true
    Use Rotational Drag=true
   ) {module=Drag002}
   /SetVariables(Particles.Unlit Scale 01:LinearColor=&MultiplyLinearColor(
     Linear Color=$Particles.Initial.Unlit Color
     Scale Factor=&MakeVector4_FromLinearColor(LinearColor=&MakeCustomLinearColorFromBool(
       Bool=&Boolean_OrOperation(Bool1=$User.bIsFrontEnd, Bool2=$User.bIsFrontEndPreview)
       True Color=&MultiplyLinearColor(
        Linear Color=&ColorFromCurve(ColorCurve=DataInterface() {propertyValues {
           AlphaCurve {keys [<at 0.0057 1 RCIM_Cubic>]}
           BlueCurve {keys [<at 0 0.3259>,<at 0.666 0.7333>]}
           GreenCurve {keys [<at 0 0.207>,<at 0.666 0.6>]}
           RedCurve {keys [<at 0 1>,<at 0.666 1>]}
          }})
        Scale Factor=&MakeVector4(
         X=&FNTimeOfDayScalar(
          Night Scale=0.8
          Morning Scale=0.8
          Day Scale=&FN_FrontEndFloatModifier(Gameplay=1.2, Front End=0.85)
          Evening Scale=0.5
         )
         Y=&FNTimeOfDayScalar(
          Night Scale=0.8
          Morning Scale=0.8
          Day Scale=&FN_FrontEndFloatModifier(Gameplay=1.2, Front End=0.85)
          Evening Scale=0.5
         )
         Z=&FNTimeOfDayScalar(
          Night Scale=0.8
          Morning Scale=0.8
          Day Scale=&FN_FrontEndFloatModifier(Gameplay=1.2, Front End=0.85)
          Evening Scale=0.5
         )
         W=&FloatFromCurve(
          FloatCurve=DataInterface() {propertyValues {Curve {keys [
              {arriveTangent=-0.0888, interpMode=RCIM_Cubic, leaveTangent=-0.0888, tangentMode=RCTM_User, time=0, value=0}
              {arriveTangent=0.12, interpMode=RCIM_Cubic, leaveTangent=0.12, tangentMode=RCTM_User, time=0.15, value=1}
              {arriveTangent=-0.0405, interpMode=RCIM_Cubic, leaveTangent=-0.0405, tangentMode=RCTM_User, time=1, value=0}
             ]}}}
          Scale Curve=&FNTimeOfDayScalar(
           Night Scale=1
           Morning Scale=1
           Day Scale=&FN_FrontEndFloatModifier(Gameplay=1, Front End=0.65)
           Evening Scale=1
          )
         )
        )
       )
       False Color=&MultiplyLinearColor(
        Linear Color=&ColorFromCurve(ColorCurve=DataInterface() {propertyValues {
           AlphaCurve {keys [<at 0.0057 1 RCIM_Cubic>]}
           BlueCurve {keys [<at 0 0.8149>,<at 0.666 0.7333>]}
           GreenCurve {keys [<at 0 0.5175>,<at 0.666 0.6>]}
           RedCurve {keys [<at 0 2.5>,<at 0.666 1>]}
          }})
        Scale Factor=&MakeVector4(
         X=&FNTimeOfDayScalar(
          Night Scale=0.8
          Morning Scale=0.8
          Day Scale=&FN_FrontEndFloatModifier(Gameplay=1.2, Front End=0.85)
          Evening Scale=0.5
         )
         Y=&FNTimeOfDayScalar(
          Night Scale=0.8
          Morning Scale=0.8
          Day Scale=&FN_FrontEndFloatModifier(Gameplay=1.2, Front End=0.85)
          Evening Scale=0.5
         )
         Z=&FNTimeOfDayScalar(
          Night Scale=0.8
          Morning Scale=0.8
          Day Scale=&FN_FrontEndFloatModifier(Gameplay=1.2, Front End=0.85)
          Evening Scale=0.5
         )
         W=&FloatFromCurve(
          FloatCurve=DataInterface() {propertyValues {Curve {keys [
              {arriveTangent=-0.0888, interpMode=RCIM_Cubic, leaveTangent=-0.0888, tangentMode=RCTM_User, time=0, value=0}
              {arriveTangent=0.12, interpMode=RCIM_Cubic, leaveTangent=0.12, tangentMode=RCTM_User, time=0.15, value=1}
              {arriveTangent=-0.0405, interpMode=RCIM_Cubic, leaveTangent=-0.0405, tangentMode=RCTM_User, time=1, value=0}
             ]}}}
          Scale Curve=&FNTimeOfDayScalar(
           Night Scale=1
           Morning Scale=1
           Day Scale=&FN_FrontEndFloatModifier(Gameplay=1, Front End=0.65)
           Evening Scale=1
          )
         )
        )
       )
      ))
    )) {enabled=false, isSetParameters=true}
   &SolveForcesAndVelocity(Clamp Velocity=false, Limit Acceleration=false)
   &DynamicMaterialParameters(
    Index 0 Param 1=&Add_Float(
     A=&RandomRangeFloat(Minimum=0, Maximum=1)
     B=&FloatFromCurve(
      FloatCurve=DataInterface() {propertyValues {Curve {keys [
          {arriveTangent=2.7533, interpMode=RCIM_Cubic, leaveTangent=1.8958, tangentMode=RCTM_User, time=0, value=1}
          {arriveTangent=0.8292, interpMode=RCIM_Cubic, leaveTangent=0.8292, tangentMode=RCTM_User, time=0.321, value=1.3453}
          {arriveTangent=0.0529, interpMode=RCIM_Cubic, leaveTangent=0.2021, tangentMode=RCTM_User, time=1, value=1.5}
         ]}}}
      CurveIndex=&Add_Float(A=&Multiply_Float(A=1, B=0.5), B=$Particles.NormalizedAge)
     )
    )
    Index 0 Param 2=&FloatFromCurve(FloatCurve=DataInterface() {propertyValues {Curve {keys [
         {arriveTangent=0.0453, interpMode=RCIM_Cubic, tangentMode=RCTM_Break, time=0, value=-1}
         {arriveTangent=0.0604, interpMode=RCIM_Cubic, leaveTangent=0.0604, tangentMode=RCTM_User, time=1, value=-1}
        ]}}})
    Index 0 Param 3=0.5
    Index 0 Param 4=&FloatFromCurve(
     FloatCurve=DataInterface() {propertyValues {Curve {keys [<at 0.5 1>,<at 1 0>]}}}
     Scale Curve=&FNTimeOfDayScalar(
      Night Scale=2
      Morning Scale=2
      Day Scale=&FN_FrontEndFloatModifier(Gameplay=1, Front End=0.5)
      Evening Scale=3
     )
    )
    Index 2 Param 1=&FloatFromCurve(FloatCurve=DataInterface() {propertyValues {Curve {keys [
         {interpMode=RCIM_Cubic, leaveTangent=4.2485, tangentMode=RCTM_Break, time=0, value=0}
         {arriveTangent=1.0548, interpMode=RCIM_Cubic, leaveTangent=1.0548, tangentMode=RCTM_User, time=0.17, value=0.5715}
         <at 1 1>
        ]}}})
    Index 2 Param 2=&FloatFromCurve(
     FloatCurve=DataInterface() {propertyValues {Curve {keys [
         {arriveTangent=1.9056, interpMode=RCIM_Cubic, leaveTangent=1.9056, tangentMode=RCTM_User, time=0, value=0.2}
         {arriveTangent=0.0259, interpMode=RCIM_Cubic, leaveTangent=0.0259, tangentMode=RCTM_User, time=1, value=1.4}
        ]}}}
     Scale Curve=&RandomRangeFloat(Minimum=0.2, Maximum=0.5)
    )
    Index 2 Param 3=&FloatFromCurve(FloatCurve=DataInterface() {propertyValues {Curve {keys [
         {interpMode=RCIM_Cubic, leaveTangent=2.3064, tangentMode=RCTM_Break, time=0, value=1}
         {interpMode=RCIM_Cubic, leaveTangent=11.1233, tangentMode=RCTM_Break, time=1, value=2}
        ]}}})
    Index 2 Param 4=&FloatFromCurve(
     FloatCurve=DataInterface() {propertyValues {Curve {keys [
         {arriveTangent=-0.6867, interpMode=RCIM_Cubic, leaveTangent=-0.6867, tangentMode=RCTM_User, time=0, value=1}
         {interpMode=RCIM_Cubic, leaveTangent=-1.2956, tangentMode=RCTM_Break, time=1, value=0.75}
        ]}}}
     Scale Curve=1.6
    )
    Module.Param0WriteEnabled=true
    Module.Param0WriteEnabled2=true
    Module.Param1WriteEnabled=false
    Module.Param1WriteEnabled2=true
    Module.Param2WriteEnabled=true
    Module.Param2WriteEnabled2=true
    Module.Param3WriteEnabled=true
    Module.Param3WriteEnabled2=true
    Write Parameter Index 2=true
   )
  ]
  rendererClasses [&NiagaraSpriteRendererProperties]
  renderers [
   {rendererClass=&NiagaraSpriteRendererProperties}
   {rendererClass=&NiagaraSpriteRendererProperties, rendererIndex=1}
  ]
 }
 Pointy1 {
  emitterUpdateScript [
   &EmitterState(
    Life Cycle Mode=::NewEnumerator1
    Loop Behavior=::NewEnumerator1
    Loop Delay=&RandomRangeFloat(Minimum=2, Maximum=4)
    Loop Duration Mode=::NewEnumerator1
    UseLoopDelay=false
   )
   &SpawnBurst_Instantaneous(Spawn Count=32, Use Spawn Probability=false, UseLoopCountLimit=false)
  ]
  particleSpawnScript [
   &InitializeParticle(
    Color=&FNTimeOfDayColor(
     Night Color=<rgba 0.705 0.41 1 1>
     Morning Color=<rgba 0.4642 0.3999 1 1>
     Day Color=<rgba 2 0.4799 1.933 0.4>
     Evening Color=<rgba 0.7663 1.5907 2 0.4>
    )
    Lifetime Max=1.2
    Lifetime Min=0.3
    Lifetime Mode=::NewEnumerator1
    Sprite Size=<20 30>
    Sprite Size Mode=::NewEnumerator4
    UsePositionOffset=false
    Write Material Random=true
   )
   &ApplyInitialForces()
   &AddVelocityFromPoint(
    Module.Use Falloff Distance=false
    Origin Offset=&RandomRangeVector(Minimum=<0 0 0>, Maximum=<0 0 0>)
    Velocity Strength=&RandomRangeFloat(Minimum=150, Maximum=400)
   )
   &SphereLocation(
    MaskSpawnGroup=false
    Offset=<0 0 -10>
    Offset Coordinate Space:ENiagaraCoordinateSpace=::World
    Override Seed=true
    Random Seed=20
    Sphere Radius=5
    Surface Only=false
    UseNonUniformScale=false
    UseOffset=true
    UseSphereOrientation=false
   )
   /SetVariables(Particles.Unlit Color:LinearColor=&MakeCustomLinearColorFromBool(
     Bool=&Boolean_OrOperation(Bool1=$User.bIsFrontEnd, Bool2=$User.bIsFrontEndPreview)
     True Color=<rgba 1.2 1.2 1.2 1>
     False Color=<rgba 1 0.851 0.9591 1>
    )) {isSetParameters=true, module=SetVariables_#2}
  ]
  particleUpdateScript [
   &ParticleState()
   &ScaleSpriteSize(Uniform Curve Sprite Scale=DataInterface() {propertyValues {Curve {keys [<at 0.001 0.5>,<at 0.2948 1.1652>,<at 1.0009 0.8072>]}}})
   &ScaleColor(
    Linear Color Curve=DataInterface() {propertyValues {
      AlphaCurve {keys [<at 0.0057 1 RCIM_Cubic>]}
      BlueCurve {keys [<at 0 0.6519>,<at 0.666 0.7333>]}
      GreenCurve {keys [<at 0 0.414>,<at 0.666 0.6>]}
      RedCurve {keys [<at 0 2>,<at 0.666 1>]}
     }}
    Scale Mode=::NewEnumerator2
   ) {module=ScaleColor001}
   &ScaleColor(
    Scale Alpha=&FloatFromCurve(
     FloatCurve=DataInterface() {propertyValues {Curve {keys [
         {arriveTangent=-0.0888, interpMode=RCIM_Cubic, leaveTangent=-0.0888, tangentMode=RCTM_User, time=0, value=0}
         {arriveTangent=0.12, interpMode=RCIM_Cubic, leaveTangent=0.12, tangentMode=RCTM_User, time=0.15, value=0.9847}
         {arriveTangent=-0.0405, interpMode=RCIM_Cubic, leaveTangent=-0.0405, tangentMode=RCTM_User, time=0.6243, value=0}
        ]}}}
     Scale Curve=&FNTimeOfDayScalar(
      Night Scale=1
      Morning Scale=1
      Day Scale=&FN_FrontEndFloatModifier(Gameplay=1, Front End=1)
      Evening Scale=1
     )
    )
    Scale RGB=&VectorFromFloat(Value=&FNTimeOfDayScalar(
      Night Scale=0.8
      Morning Scale=0.8
      Day Scale=&FN_FrontEndFloatModifier(Gameplay=1, Front End=1)
      Evening Scale=0.5
     ))
    ScaleA=true
    ScaleRGB=true
   ) {module=ScaleColor003}
   &Drag(Drag=3, Use Linear Drag=true, Use Rotational Drag=true) {module=Drag002}
   /SetVariables(Particles.Unlit Scale 01:LinearColor=&MultiplyLinearColor(
     Linear Color=$Particles.Initial.Unlit Color
     Scale Factor=&MakeVector4_FromLinearColor(LinearColor=&MakeCustomLinearColorFromBool(
       Bool=&Boolean_OrOperation(Bool1=$User.bIsFrontEnd, Bool2=$User.bIsFrontEndPreview)
       True Color=&MultiplyLinearColor(
        Linear Color=&ColorFromCurve(ColorCurve=DataInterface() {propertyValues {
           AlphaCurve {keys [<at 0.0057 1 RCIM_Cubic>]}
           BlueCurve {keys [<at 0 0.3259>,<at 0.666 0.7333>]}
           GreenCurve {keys [<at 0 0.207>,<at 0.666 0.6>]}
           RedCurve {keys [<at 0 1>,<at 0.666 1>]}
          }})
        Scale Factor=&MakeVector4(
         X=&FNTimeOfDayScalar(
          Night Scale=0.8
          Morning Scale=0.8
          Day Scale=&FN_FrontEndFloatModifier(Gameplay=1.2, Front End=0.85)
          Evening Scale=0.5
         )
         Y=&FNTimeOfDayScalar(
          Night Scale=0.8
          Morning Scale=0.8
          Day Scale=&FN_FrontEndFloatModifier(Gameplay=1.2, Front End=0.85)
          Evening Scale=0.5
         )
         Z=&FNTimeOfDayScalar(
          Night Scale=0.8
          Morning Scale=0.8
          Day Scale=&FN_FrontEndFloatModifier(Gameplay=1.2, Front End=0.85)
          Evening Scale=0.5
         )
         W=&FloatFromCurve(
          FloatCurve=DataInterface() {propertyValues {Curve {keys [
              {arriveTangent=-0.0888, interpMode=RCIM_Cubic, leaveTangent=-0.0888, tangentMode=RCTM_User, time=0, value=0}
              {arriveTangent=0.12, interpMode=RCIM_Cubic, leaveTangent=0.12, tangentMode=RCTM_User, time=0.15, value=1}
              {arriveTangent=-0.0405, interpMode=RCIM_Cubic, leaveTangent=-0.0405, tangentMode=RCTM_User, time=1, value=0}
             ]}}}
          Scale Curve=&FNTimeOfDayScalar(
           Night Scale=1
           Morning Scale=1
           Day Scale=&FN_FrontEndFloatModifier(Gameplay=1, Front End=0.65)
           Evening Scale=1
          )
         )
        )
       )
       False Color=&MultiplyLinearColor(
        Linear Color=&ColorFromCurve(ColorCurve=DataInterface() {propertyValues {
           AlphaCurve {keys [<at 0.0057 1 RCIM_Cubic>]}
           BlueCurve {keys [<at 0 0.8149>,<at 0.666 0.7333>]}
           GreenCurve {keys [<at 0 0.5175>,<at 0.666 0.6>]}
           RedCurve {keys [<at 0 2.5>,<at 0.666 1>]}
          }})
        Scale Factor=&MakeVector4(
         X=&FNTimeOfDayScalar(
          Night Scale=0.8
          Morning Scale=0.8
          Day Scale=&FN_FrontEndFloatModifier(Gameplay=1.2, Front End=0.85)
          Evening Scale=0.5
         )
         Y=&FNTimeOfDayScalar(
          Night Scale=0.8
          Morning Scale=0.8
          Day Scale=&FN_FrontEndFloatModifier(Gameplay=1.2, Front End=0.85)
          Evening Scale=0.5
         )
         Z=&FNTimeOfDayScalar(
          Night Scale=0.8
          Morning Scale=0.8
          Day Scale=&FN_FrontEndFloatModifier(Gameplay=1.2, Front End=0.85)
          Evening Scale=0.5
         )
         W=&FloatFromCurve(
          FloatCurve=DataInterface() {propertyValues {Curve {keys [
              {arriveTangent=-0.0888, interpMode=RCIM_Cubic, leaveTangent=-0.0888, tangentMode=RCTM_User, time=0, value=0}
              {arriveTangent=0.12, interpMode=RCIM_Cubic, leaveTangent=0.12, tangentMode=RCTM_User, time=0.15, value=1}
              {arriveTangent=-0.0405, interpMode=RCIM_Cubic, leaveTangent=-0.0405, tangentMode=RCTM_User, time=1, value=0}
             ]}}}
          Scale Curve=&FNTimeOfDayScalar(
           Night Scale=1
           Morning Scale=1
           Day Scale=&FN_FrontEndFloatModifier(Gameplay=1, Front End=0.65)
           Evening Scale=1
          )
         )
        )
       )
      ))
    )) {isSetParameters=true}
   &SolveForcesAndVelocity(Clamp Velocity=false, Limit Acceleration=false)
   &ScaleSpriteSizeBySpeed(Max Scale Factor=<1 5>, Velocity Threshold=100)
   &DynamicMaterialParameters(
    Index 0 Param 1=&Add_Float(
     A=&RandomRangeFloat(Minimum=0, Maximum=1)
     B=&FloatFromCurve(
      FloatCurve=DataInterface() {propertyValues {Curve {keys [
          {arriveTangent=2.7533, interpMode=RCIM_Cubic, leaveTangent=1.8958, tangentMode=RCTM_User, time=0, value=1}
          {arriveTangent=0.8292, interpMode=RCIM_Cubic, leaveTangent=0.8292, tangentMode=RCTM_User, time=0.321, value=1.3453}
          {arriveTangent=0.0529, interpMode=RCIM_Cubic, leaveTangent=0.2021, tangentMode=RCTM_User, time=1, value=1.5}
         ]}}}
      CurveIndex=&Add_Float(A=&Multiply_Float(A=1, B=0.5), B=$Particles.NormalizedAge)
     )
    )
    Index 0 Param 2=&FloatFromCurve(FloatCurve=DataInterface() {propertyValues {Curve {keys [
         {arriveTangent=0.0453, interpMode=RCIM_Cubic, tangentMode=RCTM_Break, time=0, value=-1}
         {arriveTangent=0.0604, interpMode=RCIM_Cubic, leaveTangent=0.0604, tangentMode=RCTM_User, time=1, value=-1}
        ]}}})
    Index 0 Param 3=&FloatFromCurve(FloatCurve=DataInterface() {propertyValues {Curve {keys [
         {arriveTangent=-53.9199, interpMode=RCIM_Cubic, leaveTangent=-8.9727, tangentMode=RCTM_User, time=0, value=1}
         {arriveTangent=0.2347, interpMode=RCIM_Cubic, leaveTangent=0.2723, tangentMode=RCTM_Break, time=0.25, value=0}
        ]}}})
    Index 0 Param 4=&FloatFromCurve(
     FloatCurve=DataInterface() {propertyValues {Curve {keys [<at 0.5 1>,<at 1 0>]}}}
     Scale Curve=&FNTimeOfDayScalar(
      Night Scale=2
      Morning Scale=2
      Day Scale=&FN_FrontEndFloatModifier(Gameplay=1, Front End=0.5)
      Evening Scale=3
     )
    )
    Index 2 Param 1=&FloatFromCurve(FloatCurve=DataInterface() {propertyValues {Curve {keys [
         {interpMode=RCIM_Cubic, leaveTangent=4.2485, tangentMode=RCTM_Break, time=0, value=0}
         {arriveTangent=1.0548, interpMode=RCIM_Cubic, leaveTangent=1.0548, tangentMode=RCTM_User, time=0.17, value=0.5715}
         <at 1 1>
        ]}}})
    Index 2 Param 2=&FloatFromCurve(
     FloatCurve=DataInterface() {propertyValues {Curve {keys [
         {arriveTangent=1.9056, interpMode=RCIM_Cubic, leaveTangent=1.9056, tangentMode=RCTM_User, time=0, value=0.2}
         {arriveTangent=0.0259, interpMode=RCIM_Cubic, leaveTangent=0.0259, tangentMode=RCTM_User, time=1, value=1.4}
        ]}}}
     Scale Curve=&RandomRangeFloat(Minimum=0.2, Maximum=0.5)
    )
    Index 2 Param 3=&FloatFromCurve(FloatCurve=DataInterface() {propertyValues {Curve {keys [
         {interpMode=RCIM_Cubic, leaveTangent=2.3064, tangentMode=RCTM_Break, time=0, value=1}
         {interpMode=RCIM_Cubic, leaveTangent=11.1233, tangentMode=RCTM_Break, time=1, value=2}
        ]}}})
    Index 2 Param 4=&FloatFromCurve(
     FloatCurve=DataInterface() {propertyValues {Curve {keys [
         {arriveTangent=-0.6867, interpMode=RCIM_Cubic, leaveTangent=-0.6867, tangentMode=RCTM_User, time=0, value=1}
         {interpMode=RCIM_Cubic, leaveTangent=-1.2956, tangentMode=RCTM_Break, time=1, value=0.75}
        ]}}}
     Scale Curve=1.6
    )
    Module.Param0WriteEnabled=true
    Module.Param0WriteEnabled2=true
    Module.Param1WriteEnabled=false
    Module.Param1WriteEnabled2=true
    Module.Param2WriteEnabled=false
    Module.Param2WriteEnabled2=true
    Module.Param3WriteEnabled=true
    Module.Param3WriteEnabled2=true
    Write Parameter Index 2=true
   )
  ]
  rendererClasses [&NiagaraSpriteRendererProperties]
  renderers [
   {rendererClass=&NiagaraSpriteRendererProperties}
   {rendererClass=&NiagaraSpriteRendererProperties, rendererIndex=1}
  ]
 }
 Pointy2 {
  emitterUpdateScript [
   &EmitterState(
    Life Cycle Mode=::NewEnumerator1
    Loop Behavior=::NewEnumerator1
    Loop Delay=&RandomRangeFloat(Minimum=2, Maximum=4)
    Loop Duration Mode=::NewEnumerator1
    UseLoopDelay=false
   )
   &SpawnBurst_Instantaneous(Spawn Count=7, Use Spawn Probability=false, UseLoopCountLimit=false)
  ]
  particleSpawnScript [
   &InitializeParticle(
    Color=&FNTimeOfDayColor(
     Night Color=<rgba 2.775 0.55 5 0.5>
     Morning Color=<rgba 5 1.5562 3.0382 0.5>
     Day Color=<rgba 1.0249 0.5498 1.5 0.5>
     Evening Color=<rgba 4.5598 2.1676 5 0.5>
    )
    Lifetime Max=1.2
    Lifetime Min=0.3
    Lifetime Mode=::NewEnumerator1
    Sprite Size=<60 30>
    Sprite Size Mode=::NewEnumerator4
    UsePositionOffset=false
    Write Material Random=true
   )
   &ApplyInitialForces()
   &AddVelocityFromPoint(
    Module.Use Falloff Distance=false
    Origin Offset=&RandomRangeVector(Minimum=<0 0 0>, Maximum=<0 0 0>)
    Velocity Strength=&RandomRangeFloat(Minimum=150, Maximum=400)
   )
   &SphereLocation(
    MaskSpawnGroup=false
    Offset=<0 0 -10>
    Offset Coordinate Space:ENiagaraCoordinateSpace=::World
    Override Seed=true
    Random Seed=20
    Sphere Radius=5
    Surface Only=false
    UseNonUniformScale=false
    UseOffset=true
    UseSphereOrientation=false
   )
   /SetVariables(Particles.Unlit Color:LinearColor=&MakeCustomLinearColorFromBool(
     Bool=&Boolean_OrOperation(Bool1=$User.bIsFrontEnd, Bool2=$User.bIsFrontEndPreview)
     True Color=<rgba 1.2 1.2 1.2 1>
     False Color=<rgba 1 0.851 0.9591 1>
    )) {isSetParameters=true, module=SetVariables_#2}
  ]
  particleUpdateScript [
   &ParticleState()
   &ScaleSpriteSize(Uniform Curve Sprite Scale=DataInterface() {propertyValues {Curve {keys [<at 0.001 0.5>,<at 0.2948 1.1652>,<at 1.0009 0.8072>]}}})
   &ScaleColor(
    Linear Color Curve=DataInterface() {propertyValues {
      AlphaCurve {keys [<at 0.0057 1 RCIM_Cubic>]}
      BlueCurve {keys [<at 0 0.6519>,<at 0.666 0.7333>]}
      GreenCurve {keys [<at 0 0.414>,<at 0.666 0.6>]}
      RedCurve {keys [<at 0 2>,<at 0.666 1>]}
     }}
    Scale Mode=::NewEnumerator2
   ) {module=ScaleColor001}
   &ScaleColor(
    Scale Alpha=&FloatFromCurve(
     FloatCurve=DataInterface() {propertyValues {Curve {keys [
         {arriveTangent=-0.0888, interpMode=RCIM_Cubic, leaveTangent=-0.0888, tangentMode=RCTM_User, time=0, value=0}
         {arriveTangent=0.12, interpMode=RCIM_Cubic, leaveTangent=0.12, tangentMode=RCTM_User, time=0.15, value=0.9847}
         {arriveTangent=-0.0405, interpMode=RCIM_Cubic, leaveTangent=-0.0405, tangentMode=RCTM_User, time=0.6243, value=0}
        ]}}}
     Scale Curve=&FNTimeOfDayScalar(
      Night Scale=1
      Morning Scale=1
      Day Scale=&FN_FrontEndFloatModifier(Gameplay=1, Front End=1)
      Evening Scale=1
     )
    )
    Scale RGB=&VectorFromFloat(Value=&FNTimeOfDayScalar(
      Night Scale=0.8
      Morning Scale=0.8
      Day Scale=&FN_FrontEndFloatModifier(Gameplay=1, Front End=1)
      Evening Scale=0.5
     ))
    ScaleA=true
    ScaleRGB=true
   ) {module=ScaleColor003}
   &Drag(Drag=3, Use Linear Drag=true, Use Rotational Drag=true) {module=Drag002}
   /SetVariables(Particles.Unlit Scale 01:LinearColor=&MultiplyLinearColor(
     Linear Color=$Particles.Initial.Unlit Color
     Scale Factor=&MakeVector4_FromLinearColor(LinearColor=&MakeCustomLinearColorFromBool(
       Bool=&Boolean_OrOperation(Bool1=$User.bIsFrontEnd, Bool2=$User.bIsFrontEndPreview)
       True Color=&MultiplyLinearColor(
        Linear Color=&ColorFromCurve(ColorCurve=DataInterface() {propertyValues {
           AlphaCurve {keys [<at 0.0057 1 RCIM_Cubic>]}
           BlueCurve {keys [<at 0 0.3259>,<at 0.666 0.7333>]}
           GreenCurve {keys [<at 0 0.207>,<at 0.666 0.6>]}
           RedCurve {keys [<at 0 1>,<at 0.666 1>]}
          }})
        Scale Factor=&MakeVector4(
         X=&FNTimeOfDayScalar(
          Night Scale=0.8
          Morning Scale=0.8
          Day Scale=&FN_FrontEndFloatModifier(Gameplay=1.2, Front End=0.85)
          Evening Scale=0.5
         )
         Y=&FNTimeOfDayScalar(
          Night Scale=0.8
          Morning Scale=0.8
          Day Scale=&FN_FrontEndFloatModifier(Gameplay=1.2, Front End=0.85)
          Evening Scale=0.5
         )
         Z=&FNTimeOfDayScalar(
          Night Scale=0.8
          Morning Scale=0.8
          Day Scale=&FN_FrontEndFloatModifier(Gameplay=1.2, Front End=0.85)
          Evening Scale=0.5
         )
         W=&FloatFromCurve(
          FloatCurve=DataInterface() {propertyValues {Curve {keys [
              {arriveTangent=-0.0888, interpMode=RCIM_Cubic, leaveTangent=-0.0888, tangentMode=RCTM_User, time=0, value=0}
              {arriveTangent=0.12, interpMode=RCIM_Cubic, leaveTangent=0.12, tangentMode=RCTM_User, time=0.15, value=1}
              {arriveTangent=-0.0405, interpMode=RCIM_Cubic, leaveTangent=-0.0405, tangentMode=RCTM_User, time=1, value=0}
             ]}}}
          Scale Curve=&FNTimeOfDayScalar(
           Night Scale=1
           Morning Scale=1
           Day Scale=&FN_FrontEndFloatModifier(Gameplay=1, Front End=0.65)
           Evening Scale=1
          )
         )
        )
       )
       False Color=&MultiplyLinearColor(
        Linear Color=&ColorFromCurve(ColorCurve=DataInterface() {propertyValues {
           AlphaCurve {keys [<at 0.0057 1 RCIM_Cubic>]}
           BlueCurve {keys [<at 0 0.8149>,<at 0.666 0.7333>]}
           GreenCurve {keys [<at 0 0.5175>,<at 0.666 0.6>]}
           RedCurve {keys [<at 0 2.5>,<at 0.666 1>]}
          }})
        Scale Factor=&MakeVector4(
         X=&FNTimeOfDayScalar(
          Night Scale=0.8
          Morning Scale=0.8
          Day Scale=&FN_FrontEndFloatModifier(Gameplay=1.2, Front End=0.85)
          Evening Scale=0.5
         )
         Y=&FNTimeOfDayScalar(
          Night Scale=0.8
          Morning Scale=0.8
          Day Scale=&FN_FrontEndFloatModifier(Gameplay=1.2, Front End=0.85)
          Evening Scale=0.5
         )
         Z=&FNTimeOfDayScalar(
          Night Scale=0.8
          Morning Scale=0.8
          Day Scale=&FN_FrontEndFloatModifier(Gameplay=1.2, Front End=0.85)
          Evening Scale=0.5
         )
         W=&FloatFromCurve(
          FloatCurve=DataInterface() {propertyValues {Curve {keys [
              {arriveTangent=-0.0888, interpMode=RCIM_Cubic, leaveTangent=-0.0888, tangentMode=RCTM_User, time=0, value=0}
              {arriveTangent=0.12, interpMode=RCIM_Cubic, leaveTangent=0.12, tangentMode=RCTM_User, time=0.15, value=1}
              {arriveTangent=-0.0405, interpMode=RCIM_Cubic, leaveTangent=-0.0405, tangentMode=RCTM_User, time=1, value=0}
             ]}}}
          Scale Curve=&FNTimeOfDayScalar(
           Night Scale=1
           Morning Scale=1
           Day Scale=&FN_FrontEndFloatModifier(Gameplay=1, Front End=0.65)
           Evening Scale=1
          )
         )
        )
       )
      ))
    )) {isSetParameters=true}
   &SolveForcesAndVelocity(Clamp Velocity=false, Limit Acceleration=false)
   &ScaleSpriteSizeBySpeed(Max Scale Factor=<1 5>, Velocity Threshold=100)
   &DynamicMaterialParameters(
    Index 0 Param 1=&Add_Float(
     A=&RandomRangeFloat(Minimum=0, Maximum=1)
     B=&FloatFromCurve(
      FloatCurve=DataInterface() {propertyValues {Curve {keys [
          {arriveTangent=2.7533, interpMode=RCIM_Cubic, leaveTangent=1.8958, tangentMode=RCTM_User, time=0, value=1}
          {arriveTangent=0.8292, interpMode=RCIM_Cubic, leaveTangent=0.8292, tangentMode=RCTM_User, time=0.321, value=1.3453}
          {arriveTangent=0.0529, interpMode=RCIM_Cubic, leaveTangent=0.2021, tangentMode=RCTM_User, time=1, value=1.5}
         ]}}}
      CurveIndex=&Add_Float(A=&Multiply_Float(A=1, B=0.5), B=$Particles.NormalizedAge)
     )
    )
    Index 0 Param 2=&FloatFromCurve(FloatCurve=DataInterface() {propertyValues {Curve {keys [
         {arriveTangent=0.0453, interpMode=RCIM_Cubic, tangentMode=RCTM_Break, time=0, value=-1}
         {arriveTangent=0.0604, interpMode=RCIM_Cubic, leaveTangent=0.0604, tangentMode=RCTM_User, time=1, value=-1}
        ]}}})
    Index 0 Param 3=&FN_Select_Float_from_Time_Of_Day(Morning=2.5, Day=0, Evening=2.5, Night=5)
    Index 0 Param 4=&FloatFromCurve(
     FloatCurve=DataInterface() {propertyValues {Curve {keys [<at 0.5 1>,<at 1 0>]}}}
     Scale Curve=&FNTimeOfDayScalar(
      Night Scale=2
      Morning Scale=2
      Day Scale=&FN_FrontEndFloatModifier(Gameplay=1, Front End=0.5)
      Evening Scale=3
     )
    )
    Index 2 Param 1=&FloatFromCurve(FloatCurve=DataInterface() {propertyValues {Curve {keys [
         {interpMode=RCIM_Cubic, leaveTangent=4.2485, tangentMode=RCTM_Break, time=0, value=0}
         {arriveTangent=1.0548, interpMode=RCIM_Cubic, leaveTangent=1.0548, tangentMode=RCTM_User, time=0.17, value=0.5715}
         <at 1 1>
        ]}}})
    Index 2 Param 2=&FloatFromCurve(
     FloatCurve=DataInterface() {propertyValues {Curve {keys [
         {arriveTangent=1.9056, interpMode=RCIM_Cubic, leaveTangent=1.9056, tangentMode=RCTM_User, time=0, value=0.2}
         {arriveTangent=0.0259, interpMode=RCIM_Cubic, leaveTangent=0.0259, tangentMode=RCTM_User, time=1, value=1.4}
        ]}}}
     Scale Curve=&RandomRangeFloat(Minimum=0.2, Maximum=0.5)
    )
    Index 2 Param 3=&FloatFromCurve(FloatCurve=DataInterface() {propertyValues {Curve {keys [
         {interpMode=RCIM_Cubic, leaveTangent=2.3064, tangentMode=RCTM_Break, time=0, value=1}
         {interpMode=RCIM_Cubic, leaveTangent=11.1233, tangentMode=RCTM_Break, time=1, value=2}
        ]}}})
    Index 2 Param 4=&FloatFromCurve(
     FloatCurve=DataInterface() {propertyValues {Curve {keys [
         {arriveTangent=-0.6867, interpMode=RCIM_Cubic, leaveTangent=-0.6867, tangentMode=RCTM_User, time=0, value=1}
         {interpMode=RCIM_Cubic, leaveTangent=-1.2956, tangentMode=RCTM_Break, time=1, value=0.75}
        ]}}}
     Scale Curve=1.6
    )
    Module.Param0WriteEnabled=true
    Module.Param0WriteEnabled2=true
    Module.Param1WriteEnabled=false
    Module.Param1WriteEnabled2=true
    Module.Param2WriteEnabled=true
    Module.Param2WriteEnabled2=true
    Module.Param3WriteEnabled=true
    Module.Param3WriteEnabled2=true
    Write Parameter Index 2=true
   )
  ]
  rendererClasses [&NiagaraSpriteRendererProperties]
  renderers [
   {rendererClass=&NiagaraSpriteRendererProperties}
   {rendererClass=&NiagaraSpriteRendererProperties, rendererIndex=1}
  ]
 }
}
summary {
 systemName=NS_Emote_TealMink_BagPoof
 emitters [
  {emitterName=MainCloud, rendererClasses [&NiagaraSpriteRendererProperties]}
  {emitterName=Coin, rendererClasses [&NiagaraMeshRendererProperties, &NiagaraSpriteRendererProperties]}
  {emitterName=Pointy1, rendererClasses [&NiagaraSpriteRendererProperties]}
  {emitterName=Pointy2, rendererClasses [&NiagaraSpriteRendererProperties]}
  {emitterName=IceChunks, rendererClasses [&NiagaraMeshRendererProperties]}
 ]
 userVariables [
  User.bIsFrontEnd {type=NiagaraBool}
  User.bIsFrontEndPreview {type=NiagaraBool}
  User.bIsPainted {type=NiagaraBool}
  User.HourGlassSize {type=NiagaraFloat, defaultValue=0, description=""}
  User.PaintedColor {type=LinearColor, description="", defaultValue {struct=LinearColor, value=<rgba 1 1 1 1>}}
  User.ParaGlideLeanAlpha {type=NiagaraFloat, defaultValue=0, description=""}
  User.RibbonIDOffset {type=NiagaraInt32}
  User.SkeletalMesh {
   type=NiagaraDataInterfaceSkeletalMesh
   defaultValue {
    struct=NiagaraExt_VariableValue_DataInterface
    value {dataInterface=&self:NiagaraDataInterfaceSkeletalMesh_0}
   }
  }
  User.VariantSwitch {type=NiagaraFloat, defaultValue=0, description=""}
 ]
}
