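# Niagara system for the bag poof in the TealMink emote: a one-burst puff with flying particles, mist and a ground flash.
# Dirt_TrailHead throws stretched sprites that emit location events for fading Dirt_Trail ribbons, with User.VariantSwitch blending the sprites between warm and green color ramps.
# Mist adds pink flipbook smoke, WindPlanes a short expanding mesh, and BlowingParticles a fading decal.
# The mesh and decal emitters compute time-of-day tints.
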NiagaraSystem /BRCosmetics/Animation/Game/MainPlayer/Emotes/TealMink/FX/NS_Emote_TealMink_BagPoof2
dependencies {
 usedDataInterfaces [&NiagaraDataInterfaceCurve, &NiagaraDataInterfaceColorCurve, &NiagaraDataInterfaceSkeletalMesh]
 usedDynamicInputs []
 usedModules [
  &SystemState
  &self:SystemSpawnScript.SystemScriptSource.SystemScriptGraph.NiagaraNodeAssignment_0.SetVariables_#1
 ]
}
emitters {
 BlowingParticles {
  emitterUpdateScript [
   &EmitterState(Life Cycle Mode=::NewEnumerator1, Loop Behavior=::NewEnumerator1, Loop Duration=4.5, UseLoopDelay=false)
   &SpawnBurst_Instantaneous(Use Spawn Probability=false, UseLoopCountLimit=false)
   /SetVariables(Emitter.float:NiagaraFloat=&FloatFromCurve(FloatCurve=DataInterface() {propertyValues {Curve {keys [<at 0 1>,<at 1 0>]}}}, CurveIndex=$Emitter.NormalizedLoopAge, Scale Curve=0.2)) {isSetParameters=true}
   /SetVariables(Emitter.Linear Color:LinearColor=&FNTimeOfDayColor(
     Night Color=<rgba 1 0.2689 0.3071 1>
     Morning Color=<rgba 1 0.2423 0.5265 1>
     Day Color=<rgba 1 0.4891 0.9128 1>
     Evening Color=<rgba 1 0.2979 0.4227 1>
    )) {isSetParameters=true}
  ]
  particleSpawnScript [&InitializeParticle(
    Lifetime Max=4.5
    Lifetime Min=4.5
    Lifetime Mode=::NewEnumerator1
    Mass Max=1.5
    Mass Mode=::NewEnumerator3
    Sprite Rotation Mode=::NewEnumerator2
    Sprite Size Mode=::NewEnumerator1
    Uniform Sprite Size Max=8
    Uniform Sprite Size Min=4
    UsePositionOffset=false
    Write Material Random=true
   )]
  particleUpdateScript [
   &ParticleState()
   /SetVariables(Particles.DecalSize:Vector3f=&Multiply_Vector(
     A=<20 250 250>
     B=&VectorFromFloat(Value=&FloatFromCurve(FloatCurve=DataInterface() {propertyValues {Curve {keys [
           {arriveTangent=4.1785, interpMode=RCIM_Cubic, leaveTangent=4.1785, tangentMode=RCTM_User, time=0, value=0.5}
           {arriveTangent=1.1179, interpMode=RCIM_Cubic, leaveTangent=1.1179, tangentMode=RCTM_User, time=0.01, value=0.8}
           {arriveTangent=0.0224, interpMode=RCIM_Cubic, leaveTangent=0.0224, tangentMode=RCTM_User, time=0.05, value=1}
          ]}}}))
    )) {isSetParameters=true}
   &ScaleColor(
    Scale Alpha=&FloatFromCurve(FloatCurve=DataInterface() {propertyValues {Curve {keys [<at 0 1>,<at 1 0>]}}})
    ScaleA=true
    ScaleRGB=true
   )
  ]
  rendererClasses [&NiagaraDecalRendererProperties]
  renderers [{rendererClass=&NiagaraDecalRendererProperties, rendererIndex=0}]
 }
 Dirt_Trail {
  emitterUpdateScript [&EmitterState()]
  particleSpawnScript [&InitializeParticle(
    Color=<rgba 1 1 1 2>
    Lifetime Max=0.9
    Lifetime Min=0.4
    Lifetime Mode=::NewEnumerator1
    Ribbon Width=1.7
    Ribbon Width Mode=::NewEnumerator1
    UsePositionOffset=false
    Write Material Random=true
   )]
  particleUpdateScript [
   &ParticleState()
   &ScaleRibbonWidth(Ribbon Width Scale=&FloatFromCurve(FloatCurve=DataInterface() {propertyValues {Curve {keys [
         {tangentMode=RCTM_User, time=0, value=1}
         {arriveTangent=-0.6933, interpMode=RCIM_Cubic, leaveTangent=-0.6933, tangentMode=RCTM_User, time=1, value=0}
        ]}}}))
   &ScaleColor(
    Scale Alpha=&FloatFromCurve(FloatCurve=DataInterface() {propertyValues {Curve {keys [<at 0 1>,<at 0.6 0.6>,<at 1 0>]}}})
    Scale RGB=&MakeVectorFromLinearColorRGB(Color=&ColorFromCurve(ColorCurve=DataInterface() {propertyValues {
        AlphaCurve {keys []}
        BlueCurve {keys [<at 0.0734 0.7691 RCIM_Cubic>,<at 0.1716 1>,<at 0.4113 0.9488 RCIM_Cubic>]}
        GreenCurve {keys [<at 0.0734 0.6421 RCIM_Cubic>,<at 0.1716 0.3893>,<at 0.4113 0.9392 RCIM_Cubic>]}
        RedCurve {keys [<at 0.0734 1 RCIM_Cubic>,<at 0.1716 0.666>,<at 0.4113 1 RCIM_Cubic>]}
       }}))
    ScaleA=true
    ScaleRGB=true
   )
   &ScaleColor(
    Scale Alpha=&FloatFromCurve(FloatCurve=DataInterface() {propertyValues {Curve {keys [<at 0 0.5>,<at 0.8 0>]}}}, CurveIndex=$System.Age)
    ScaleA=true
    ScaleRGB=true
   ) {module=ScaleColor001}
  ]
  rendererClasses [&NiagaraRibbonRendererProperties]
  renderers [{rendererClass=&NiagaraRibbonRendererProperties}]
 }
 Dirt_TrailHead {
  emitterUpdateScript [
   &EmitterState(Inactive Response=::NewEnumerator1, Life Cycle Mode=::NewEnumerator1, Loop Behavior=::NewEnumerator1, Loop Duration=1, UseLoopDelay=false)
   &SpawnBurst_Instantaneous(Spawn Count=40, Use Spawn Probability=false, UseLoopCountLimit=false)
   &SpawnRate(SpawnRate=10, Use Spawn Probability=false) {enabled=false}
  ]
  particleSpawnScript [
   &InitializeParticle(
    Lifetime Max=0.8
    Lifetime Min=0.35
    Lifetime Mode=::NewEnumerator1
    Position Offset=$System.NS Position
    Sprite Size Mode=::NewEnumerator1
    Uniform Sprite Size Max=0.5
    Uniform Sprite Size Min=0.3
    UsePositionOffset=true
    Write Material Random=true
   )
   &ShapeLocation(
    MaskSpawnGroup=false
    Override Seed=false
    Rotation Angle=0.2
    Rotation Axis=<0 1 0>
    Rotation Mode=::NewEnumerator1
    Sphere Radius=5
    UseOwnerScale=true
   )
   &AddVelocity(
    Module.Use Falloff Distance=false
    Origin Offset=$System.NS Position
    Override Seed=false
    Velocity Mode=::NewEnumerator1
    Velocity Speed=&RandomRangeFloat(Minimum=100, Maximum=1000)
   )
   &AddVelocity(Velocity=<0 0 1>) {module=AddVelocity001}
  ]
  particleUpdateScript [
   &ParticleState()
   &SpriteRotationRate(Rotation Rate=&Multiply_Float(A=&RandomRangeFloat(Minimum=-1, Maximum=1), B=800))
   &Drag(Drag=&RandomRangeFloat(Minimum=3, Maximum=5), Use Linear Drag=true, Use Rotational Drag=true)
   &GravityForce(Gravity=&RandomRangeVector(Minimum=<0 0 -300>, Maximum=<0 0 -600>))
   &SolveForcesAndVelocity(Clamp Velocity=false, Limit Acceleration=false)
   &ScaleSpriteSizeBySpeed(Max Scale Factor=<0.5 8>, Min Scale Factor=<0.4 5>)
   &GenerateLocationEvent(Use Event Probability=false, UseEventDelay=false)
   &ScaleColor(
    Scale RGB=/Niagara/DynamicInputs/Math/LerpMultipleVectors(
     Alpha:NiagaraFloat=$User.VariantSwitch
     V 1:Vector3f=&MakeVectorFromLinearColorRGB(Color=&ColorFromCurve(ColorCurve=DataInterface() {propertyValues {
         AlphaCurve {keys []}
         BlueCurve {keys [<at 0 1.9927 RCIM_Cubic>,<at 0.132 1>,<at 0.52 0.25 RCIM_Cubic>]}
         GreenCurve {keys [<at 0 0.9332 RCIM_Cubic>,<at 0.132 0.5153>,<at 0.52 0.063 RCIM_Cubic>]}
         RedCurve {keys [<at 0 2 RCIM_Cubic>,<at 0.132 0.9338>,<at 0.52 0.2489 RCIM_Cubic>]}
        }}))
     V 2:Vector3f=&MakeVectorFromLinearColorRGB(Color=&ColorFromCurve(ColorCurve=DataInterface() {propertyValues {
         AlphaCurve {keys []}
         BlueCurve {keys [<at 0 0.2092 RCIM_Cubic>,<at 0.132 0.3519>,<at 0.52 0.0824 RCIM_Cubic>]}
         GreenCurve {keys [<at 0 0.2552 RCIM_Cubic>,<at 0.132 1>,<at 0.52 0.1406 RCIM_Cubic>]}
         RedCurve {keys [<at 0 0.0086 RCIM_Cubic>,<at 0.132 0>,<at 0.52 0.0161 RCIM_Cubic>]}
        }}))
     V 3:Vector3f=?
     V 4:Vector3f=?
     V 5:Vector3f=?
     Vector Count:ENiagara_MultipleLerpCount=::NewEnumerator3
    )
    ScaleA=true
    ScaleRGB=true
   )
  ]
  rendererClasses [&NiagaraSpriteRendererProperties]
  renderers [{rendererClass=&NiagaraSpriteRendererProperties}]
 }
 Mist {
  emitterUpdateScript [
   &EmitterState(Life Cycle Mode=::NewEnumerator1, Loop Behavior=::NewEnumerator1, UseLoopDelay=false)
   &SpawnBurst_Instantaneous(Spawn Count=14, Use Spawn Probability=false)
  ]
  particleSpawnScript [
   /SetVariables(Particles.Random=&UniformRangedFloat(Minimum=0, Maximum=1)) {isSetParameters=true}
   &InitializeParticle(
    Color=<rgba 1 0.3003 0.9888 0.3>
    Lifetime=&UniformRangedFloat(Minimum=0.9818, Maximum=1.7036)
    Sprite Rotation Mode=::NewEnumerator2
    Sprite Size Mode=::NewEnumerator3
    Uniform Sprite Size=&UniformRangedFloat(Minimum=50, Maximum=100)
    UsePositionOffset=false
    Write Material Random=true
   ) {module=InitializeParticle001}
   &DynamicMaterialParameters(
    Index 0 Param 1=&UniformRangedFloat(Minimum=1, Maximum=2)
    Index 0 Param 2=&UniformRangedFloat(Minimum=3, Maximum=5)
    Index 0 Param 3=1
    Module.Param0WriteEnabled=true
    Module.Param1WriteEnabled=true
    Module.Param2WriteEnabled=true
    Module.Param3WriteEnabled=false
   ) {enabled=false}
   &AddVelocityInCone(
    Cone Angle=80.4364
    Cone Axis=<0 0 1>
    Cone Axis Coordinate Space=::World
    Module.Use Velocity Falloff On Cone Axis=false
    Velocity Strength=&UniformRangedFloat(Minimum=150, Maximum=300)
   )
  ]
  particleUpdateScript [
   &ParticleState()
   &ScaleSpriteSize(Scale Factor=&Vector2DFromFloat(Value=&FloatFromCurve(FloatCurve=DataInterface() {propertyValues {Curve {keys [<at -0.0046 0.5>,<at 1.0011 2>]}}}, Scale Curve=1.5)))
   /Niagara/Modules/Update/SubUV/SubUVAnimation(Number Of Frames=&FloatFromCurve(FloatCurve=DataInterface() {propertyValues {Curve {keys [<at -0.0046 0>,<at 1.0011 64>]}}}))
   &ScaleColor(
    Scale Alpha=&FloatFromCurve(FloatCurve=DataInterface() {propertyValues {Curve {keys [<at 0 0 RCIM_Cubic>, {arriveTangent=0.002, interpMode=RCIM_Cubic, leaveTangent=0.002, time=0.1942, value=2}, <at 1.0011 0.002 RCIM_Cubic>]}}})
    ScaleA=true
    ScaleRGB=true
   )
   &Drag(Use Linear Drag=true, Use Rotational Drag=false)
   &GravityForce(Gravity=&Multiply_Vector(A=<0 0 -980>, B=<0 0 0.5>))
   &SolveForcesAndVelocity(Clamp Velocity=false, Limit Acceleration=false)
   &DynamicMaterialParameters(
    Index 0 Param 1=&UniformRangedFloat(Minimum=4, Maximum=7)
    Index 0 Param 2=8
    Index 0 Param 3=1
    Index 0 Param 4=0.01
    Module.Param0WriteEnabled=true
    Module.Param1WriteEnabled=true
    Module.Param2WriteEnabled=true
    Module.Param3WriteEnabled=true
   ) {module=DynamicMaterialParameters001}
  ]
  rendererClasses [&NiagaraSpriteRendererProperties]
  renderers [{rendererClass=&NiagaraSpriteRendererProperties}]
 }
 WindPlanes {
  emitterUpdateScript [
   &EmitterState(Inactive Response=::NewEnumerator1, Life Cycle Mode=::NewEnumerator1, Loop Behavior=::NewEnumerator1, Loop Duration=4.5, UseLoopDelay=false)
   &SpawnBurst_Instantaneous(Use Spawn Probability=false, UseLoopCountLimit=false)
   /SetVariables(Emitter.Linear Color:LinearColor=&FNTimeOfDayColor(
     Night Color=<rgba 1.5943 0.4975 2.5 1>
     Morning Color=<rgba 1 0.2453 0.6479 1>
     Day Color=<rgba 2 0.5761 1.2749 1>
     Evening Color=<rgba 1 0.5229 0.7855 1>
    )) {isSetParameters=true}
  ]
  particleSpawnScript [
   &InitializeParticle(
    Color=<rgba 0 0.6493 1 0.4365>
    Lifetime=0.4
    Mesh Scale Mode=::NewEnumerator3
    Mesh Uniform Scale=1.1
    Position Offset=<0 0 0>
    UsePositionOffset=true
    Write Material Random=true
   )
   &CameraOffset(Camera Offset Amount=5)
   &InitialMeshOrientation(Mesh Axis To Orient=<0 0 1>, Mesh Orientation Mode=::NewEnumerator2, Orientation Vector=<0 0 1>, Use Rotation Vector=false)
  ]
  particleUpdateScript [
   &ParticleState()
   &ScaleMeshSize(Scale Factor=&VectorFromFloat(Value=&FloatFromCurve(FloatCurve=DataInterface() {propertyValues {Curve {keys [
          {arriveTangent=3.1045, interpMode=RCIM_Cubic, leaveTangent=3.1045, leaveTangentWeight=0.8949, tangentMode=RCTM_User, time=0, value=0}
          {arriveTangentWeight=0.4784, interpMode=RCIM_Cubic, tangentMode=RCTM_User, time=1, value=1}
         ]}}})))
   &ScaleColor(
    Scale Alpha=&FloatFromCurve(
     FloatCurve=DataInterface() {propertyValues {Curve {keys [
         {interpMode=RCIM_Cubic, leaveTangentWeight=2.5689, tangentMode=RCTM_User, time=0, value=0}
         {interpMode=RCIM_Cubic, tangentMode=RCTM_User, time=0.1962, value=1}
         {interpMode=RCIM_Cubic, tangentMode=RCTM_User, time=1, value=0}
        ]}}}
     Scale Curve=&FNTimeOfDayScalar(Night Scale=1, Morning Scale=1, Evening Scale=1)
    )
    ScaleA=true
    ScaleRGB=false
   )
  ]
  rendererClasses [&NiagaraMeshRendererProperties]
  renderers [{rendererClass=&NiagaraMeshRendererProperties, rendererIndex=0}]
 }
}
summary {
 systemName=NS_Emote_TealMink_BagPoof2
 emitters [
  {emitterName=Dirt_Trail, rendererClasses [&NiagaraRibbonRendererProperties]}
  {emitterName=Dirt_TrailHead, rendererClasses [&NiagaraSpriteRendererProperties]}
  {emitterName=Mist, rendererClasses [&NiagaraSpriteRendererProperties]}
  {emitterName=WindPlanes, rendererClasses [&NiagaraMeshRendererProperties]}
  {emitterName=BlowingParticles, rendererClasses [&NiagaraDecalRendererProperties]}
 ]
 userVariables [
  User.bIsFrontEnd {type=NiagaraBool}
  User.bIsFrontEndPreview {type=NiagaraBool}
  User.bIsPainted {type=NiagaraBool}
  User.HourGlassSize {type=NiagaraFloat, defaultValue=0, description=""}
  User.PaintedColor {type=LinearColor, description="", defaultValue {struct=LinearColor, value=<rgba 1 1 1 1>}}
  User.ParaGlideLeanAlpha {type=NiagaraFloat, defaultValue=0, description=""}
  User.RibbonIDOffset {type=NiagaraInt32}
  User.VariantSwitch {type=NiagaraFloat, defaultValue=0, description=""}
 ]
}
