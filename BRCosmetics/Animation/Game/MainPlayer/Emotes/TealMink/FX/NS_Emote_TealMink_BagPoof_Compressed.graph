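# Niagara system for the TealMink emote's bag poof, a mostly pink one-shot burst.
# It layers tumbling, partly colliding coin meshes and sprites, spawn-group-masked smoke, mesh chunks, a lit flash and a fading decal.
# Trail-head sprites emit location events alongside a ribbon trail emitter, and User.VariantSwitch blends their colors.
# Time-of-day functions tint smoke, flash and coins, while front-end checks adjust coin color and flare size.
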NiagaraSystem /BRCosmetics/Animation/Game/MainPlayer/Emotes/TealMink/FX/NS_Emote_TealMink_BagPoof_Compressed
dependencies {
 usedDataInterfaces [
  &NiagaraDataInterfaceCurve
  &NiagaraDataInterfaceArrayInt32
  &NiagaraDataInterfaceSpriteRendererInfo
  &NiagaraDataInterfaceColorCurve
  &NiagaraDataInterfaceMeshRendererInfo
 ]
 usedDynamicInputs []
 usedModules [&SystemState]
}
emitters {
 BlowingParticles {
  emitterUpdateScript [
   &EmitterState(Life Cycle Mode=::NewEnumerator1, Loop Behavior=::NewEnumerator1, Loop Duration=4.5, UseLoopDelay=false)
   &SpawnBurst_Instantaneous(Use Spawn Probability=false, UseLoopCountLimit=false)
   /SetVariables(Emitter.float:NiagaraFloat=&FloatFromCurve(FloatCurve=DataInterface() {propertyValues {Curve {keys [<at 0 1>,<at 1 0>]}}}, CurveIndex=$Emitter.NormalizedLoopAge, Scale Curve=0.2)) {isSetParameters=true}
   /SetVariables(Emitter.Linear Color:LinearColor=&FNTimeOfDayColor(
     Night Color=<rgba 1 0.2689 0.3071 1>
     Morning Color=<rgba 1 0.2423 0.5265 1>
     Day Color=<rgba 1 0.4891 0.9128 1>
     Evening Color=<rgba 1 0.2979 0.4227 1>
    )) {isSetParameters=true}
  ]
  particleSpawnScript [&InitializeParticle(
    Lifetime Max=4.5
    Lifetime Min=4.5
    Lifetime Mode=::NewEnumerator1
    Mass Max=1.5
    Mass Mode=::NewEnumerator3
    Sprite Rotation Mode=::NewEnumerator2
    Sprite Size Mode=::NewEnumerator1
    Uniform Sprite Size Max=8
    Uniform Sprite Size Min=4
    UsePositionOffset=false
    Write Material Random=true
   )]
  particleUpdateScript [
   &ParticleState()
   /SetVariables(Particles.DecalSize:Vector3f=&Multiply_Vector(
     A=<20 250 250>
     B=&VectorFromFloat(Value=&FloatFromCurve(FloatCurve=DataInterface() {propertyValues {Curve {keys [
           {arriveTangent=4.1785, interpMode=RCIM_Cubic, leaveTangent=4.1785, tangentMode=RCTM_User, time=0, value=0.5}
           {arriveTangent=1.1179, interpMode=RCIM_Cubic, leaveTangent=1.1179, tangentMode=RCTM_User, time=0.01, value=0.8}
           {arriveTangent=0.0224, interpMode=RCIM_Cubic, leaveTangent=0.0224, tangentMode=RCTM_User, time=0.05, value=1}
          ]}}}))
    )) {isSetParameters=true}
   &ScaleColor(
    Scale Alpha=&FloatFromCurve(FloatCurve=DataInterface() {propertyValues {Curve {keys [<at 0 1>,<at 1 0>]}}})
    ScaleA=true
    ScaleRGB=true
   )
  ]
  rendererClasses [&NiagaraDecalRendererProperties]
  renderers [{rendererClass=&NiagaraDecalRendererProperties, rendererIndex=0}]
 }
 Coin {
  emitterUpdateScript [
   &EmitterState(
    Life Cycle Mode=::NewEnumerator1
    Loop Behavior=::NewEnumerator1
    Loop Delay=&RandomRangeFloat(Minimum=2, Maximum=4)
    Loop Duration Mode=::NewEnumerator1
    UseLoopDelay=false
   )
   &SpawnBurst_Instantaneous(Spawn Count=150, Use Spawn Probability=false, UseLoopCountLimit=false)
   /SetVariables(Emitter.SineWave:NiagaraFloat=&Sine(Normalized Angle=$Emitter.Age, Period=&RandomRangeFloat(Minimum=0.4, Maximum=1.1), Scale=1, Bias=0)) {isSetParameters=true}
  ]
  particleSpawnScript [
   /SetVariables(Particles.RendererChoice:NiagaraInt32=&RandomRangeInt(Minimum=0, Maximum=1)) {isSetParameters=true}
   /SetVariables(Particles.VisibilityTag=$Particles.RendererChoice) {isSetParameters=true}
   &InitializeParticle(
    Color=&MakeCustomLinearColorFromBool(
     Bool=&Boolean_OrOperation(Bool1=$User.bIsFrontEnd, Bool2=$User.bIsFrontEndPreview)
     True Color=<rgba 0.3311 1 0.6483 1>
     False Color=<rgba 1 0.9843 0.9961 1>
    )
    Lifetime Mode=::NewEnumerator1
    Mesh Scale Mode=::NewEnumerator3
    Mesh Uniform Scale=&RandomRangeFloat(Minimum=0.05, Maximum=0.03)
    Sprite Rotation Mode=::NewEnumerator2
    Sprite Size Mode=::NewEnumerator3
    Uniform Sprite Size=5
    UsePositionOffset=false
    Write Material Random=true
   )
   &ApplyInitialForces()
   &SphereLocation(
    MaskSpawnGroup=false
    Offset=<0 0 -10>
    Offset Coordinate Space:ENiagaraCoordinateSpace=::World
    Override Seed=true
    Random Seed=20
    Sphere Radius=5
    Surface Only=false
    UseNonUniformScale=false
    UseOffset=true
    UseSphereOrientation=false
   )
   &InitialMeshOrientation(Use Rotation Vector=false)
   /SetVariables(Particles.MeshOpacity:LinearColor=<rgba 0.7 0.3879 0.1268 0>, Particles.SpriteColor=<rgba 1 0.9843 0.9961 1>) {isSetParameters=true}
   &AddVelocity(
    Rotation Coordinate Space=::World
    Velocity=&RandomRangeVector(Minimum=<150 150 20>, Maximum=<-150 -150 600>)
   )
  ]
  particleUpdateScript [
   /SetVariables(Particles.SineWave:NiagaraFloat=&Sine(Normalized Angle=$Emitter.Age, Period=&RandomRangeFloat(Minimum=0.3, Maximum=0.7), Scale=1, Bias=0)) {isSetParameters=true}
   &ParticleState()
   &ScaleSpriteSize(Scale Factor=&MakeVector2D(X=$Particles.SineWave, Y=1), Scale Sprite Size Mode=::NewEnumerator2) {module=ScaleSpriteSize001}
   &SpriteRotationRate(Rotation Rate=&RandomRangeFloat(Minimum=-1000, Maximum=-500))
   &ScaleMeshSize(Scale Factor=&VectorFromFloat(Value=&FloatFromCurve(FloatCurve=DataInterface() {propertyValues {Curve {keys [<at 0.6 1>,<at 1 0>]}}})))
   &ScaleSpriteSize(Uniform Curve Sprite Scale=DataInterface() {propertyValues {Curve {keys [<at 0.001 0.5>,<at 0.2948 1.1652>,<at 1.0009 0.8072>]}}})
   &ScaleColor(
    Scale Alpha=&FloatFromCurve(
     FloatCurve=DataInterface() {propertyValues {Curve {keys [
         {arriveTangent=-0.0888, interpMode=RCIM_Cubic, leaveTangent=-0.0888, tangentMode=RCTM_User, time=0, value=0}
         {arriveTangent=0.12, interpMode=RCIM_Cubic, leaveTangent=0.12, tangentMode=RCTM_User, time=0.15, value=0.9847}
         {arriveTangent=-0.0405, interpMode=RCIM_Cubic, leaveTangent=-0.0405, tangentMode=RCTM_User, time=1, value=0}
        ]}}}
     Scale Curve=&FNTimeOfDayScalar(
      Night Scale=1
      Morning Scale=1
      Day Scale=&FN_FrontEndFloatModifier(Gameplay=1, Front End=1)
      Evening Scale=1
     )
    )
    Scale RGB=&VectorFromFloat(Value=&FNTimeOfDayScalar(
      Night Scale=0.8
      Morning Scale=0.8
      Day Scale=&FN_FrontEndFloatModifier(Gameplay=1, Front End=1)
      Evening Scale=0.5
     ))
    ScaleA=true
    ScaleRGB=true
   ) {module=ScaleColor003}
   &Drag(
    Drag=&FloatFromCurve(
     FloatCurve=DataInterface() {propertyValues {Curve {keys [
         {arriveTangent=0.7001, interpMode=RCIM_Cubic, leaveTangent=0.7001, tangentMode=RCTM_User, time=0, value=0}
         {interpMode=RCIM_Cubic, tangentMode=RCTM_User, time=1, value=1}
        ]}}}
     Scale Curve=4
    )
    Use Linear Drag=true
    Use Rotational Drag=true
   ) {module=Drag002}
   &GravityForce()
   &UpdateMeshOrientation(Rotation Vector=&Multiply_Vector(
     A=<1 1 1>
     B=&MakeVector(
      X=&MakeCustomFloatFromBool(Bool=&RandomBool(), True Float=2, False Float=-2)
      Y=&RandomRangeFloat(Minimum=0.1, Maximum=-0.1)
      Z=&RandomRangeFloat(Minimum=0.1, Maximum=-0.1)
     )
    ))
   &Collision(
    Age Colliding Particles=false
    Collision Enabled=&SetBoolByFloatComparison(
     A=&RandomRangeFloat(
      Minimum=0
      Maximum=&Multiply_FloatByInt(
       Float=1
       Integer=&SelectIntFromArray(Direct Array Index=$Particles.RendererChoice, Int Selection Array=DataInterface() {propertyValues {IntData [1,0]}})
      )
     )
     B=0.8
    )
    Control Roll On Collision=true
    KillOnCollision=false
    Module.Control Rest State Via Surface Normal=false
    Module.EnableMaxCollisionCount=false
    Module.Randomize Collision Normal Vector=false
    Restitution=1
    ScaleRateOfRotation=false
   )
   &SolveForcesAndVelocity(Clamp Velocity=false, Limit Acceleration=false)
   &DynamicMaterialParameters(
    Index 0 Param 1=&Add_Float(
     A=&RandomRangeFloat(Minimum=0, Maximum=1)
     B=&FloatFromCurve(
      FloatCurve=DataInterface() {propertyValues {Curve {keys [
          {arriveTangent=2.7533, interpMode=RCIM_Cubic, leaveTangent=1.8958, tangentMode=RCTM_User, time=0, value=1}
          {arriveTangent=0.8292, interpMode=RCIM_Cubic, leaveTangent=0.8292, tangentMode=RCTM_User, time=0.321, value=1.3453}
          {arriveTangent=0.0529, interpMode=RCIM_Cubic, leaveTangent=0.2021, tangentMode=RCTM_User, time=1, value=1.5}
         ]}}}
      CurveIndex=&Add_Float(A=&Multiply_Float(A=1, B=0.5), B=$Particles.NormalizedAge)
     )
    )
    Index 0 Param 2=&FloatFromCurve(FloatCurve=DataInterface() {propertyValues {Curve {keys [
         {arriveTangent=0.0453, interpMode=RCIM_Cubic, tangentMode=RCTM_Break, time=0, value=-1}
         {arriveTangent=0.0604, interpMode=RCIM_Cubic, leaveTangent=0.0604, tangentMode=RCTM_User, time=1, value=-1}
        ]}}})
    Index 0 Param 3=&FloatFromCurve(FloatCurve=DataInterface() {propertyValues {Curve {keys [
         {arriveTangent=-53.9199, interpMode=RCIM_Cubic, leaveTangent=-8.9727, tangentMode=RCTM_User, time=0, value=1}
         {arriveTangent=0.2347, interpMode=RCIM_Cubic, leaveTangent=0.2723, tangentMode=RCTM_Break, time=0.25, value=0}
        ]}}})
    Index 0 Param 4=&FloatFromCurve(
     FloatCurve=DataInterface() {propertyValues {Curve {keys [<at 0.5 1>,<at 1 0>]}}}
     Scale Curve=&FNTimeOfDayScalar(
      Night Scale=2
      Morning Scale=2
      Day Scale=&FN_FrontEndFloatModifier(Gameplay=1, Front End=0.5)
      Evening Scale=3
     )
    )
    Index 2 Param 1=&FloatFromCurve(FloatCurve=DataInterface() {propertyValues {Curve {keys [
         {interpMode=RCIM_Cubic, leaveTangent=4.2485, tangentMode=RCTM_Break, time=0, value=0}
         {arriveTangent=1.0548, interpMode=RCIM_Cubic, leaveTangent=1.0548, tangentMode=RCTM_User, time=0.17, value=0.5715}
         <at 1 1>
        ]}}})
    Index 2 Param 2=&FloatFromCurve(
     FloatCurve=DataInterface() {propertyValues {Curve {keys [
         {arriveTangent=1.9056, interpMode=RCIM_Cubic, leaveTangent=1.9056, tangentMode=RCTM_User, time=0, value=0.2}
         {arriveTangent=0.0259, interpMode=RCIM_Cubic, leaveTangent=0.0259, tangentMode=RCTM_User, time=1, value=1.4}
        ]}}}
     Scale Curve=&RandomRangeFloat(Minimum=0.2, Maximum=0.5)
    )
    Index 2 Param 3=&FloatFromCurve(FloatCurve=DataInterface() {propertyValues {Curve {keys [
         {interpMode=RCIM_Cubic, leaveTangent=2.3064, tangentMode=RCTM_Break, time=0, value=1}
         {interpMode=RCIM_Cubic, leaveTangent=11.1233, tangentMode=RCTM_Break, time=1, value=2}
        ]}}})
    Index 2 Param 4=&FloatFromCurve(
     FloatCurve=DataInterface() {propertyValues {Curve {keys [
         {arriveTangent=-0.6867, interpMode=RCIM_Cubic, leaveTangent=-0.6867, tangentMode=RCTM_User, time=0, value=1}
         {interpMode=RCIM_Cubic, leaveTangent=-1.2956, tangentMode=RCTM_Break, time=1, value=0.75}
        ]}}}
     Scale Curve=1.6
    )
    Module.Param0WriteEnabled=true
    Module.Param0WriteEnabled2=true
    Module.Param1WriteEnabled=false
    Module.Param1WriteEnabled2=true
    Module.Param2WriteEnabled=false
    Module.Param2WriteEnabled2=true
    Module.Param3WriteEnabled=true
    Module.Param3WriteEnabled2=true
    Write Parameter Index 2=true
   )
   &SubUVAnimation(
    Evaluation Type=::NewEnumerator0
    "Sprite Renderer "=DataInterface() {propertyValues {SpriteRenderer=&self:Coin_0.NiagaraSpriteRendererProperties_0}}
    SubUV Animation Mode:ENiagara_SubUVLookupModeV2=::NewEnumerator2
    UseEndFrame=false
    UseStartFrame=false
   )
   &ScaleColor(
    Scale RGB=&VectorFromFloat(Value=&RemapRange(Input Value To Remap=$Particles.SineWave, Input Min=-1, Input Max=1, Output Min=1, Output Max=3))
    ScaleA=true
    ScaleRGB=true
   )
  ]
  rendererClasses [&NiagaraMeshRendererProperties, &NiagaraSpriteRendererProperties]
  renderers [
   {rendererClass=&NiagaraMeshRendererProperties, rendererIndex=0}
   {rendererClass=&NiagaraSpriteRendererProperties, rendererIndex=1}
  ]
 }
 Dirt_Trail {
  emitterUpdateScript [&EmitterState()]
  particleSpawnScript [&InitializeParticle(
    Color=<rgba 1 1 1 2>
    Lifetime Max=0.9
    Lifetime Min=0.4
    Lifetime Mode=::NewEnumerator1
    Ribbon Width=1.7
    Ribbon Width Mode=::NewEnumerator1
    UsePositionOffset=false
    Write Material Random=true
   )]
  particleUpdateScript [
   &ParticleState()
   &ScaleRibbonWidth(Ribbon Width Scale=&FloatFromCurve(FloatCurve=DataInterface() {propertyValues {Curve {keys [
         {tangentMode=RCTM_User, time=0, value=1}
         {arriveTangent=-0.6933, interpMode=RCIM_Cubic, leaveTangent=-0.6933, tangentMode=RCTM_User, time=1, value=0}
        ]}}}))
   &ScaleColor(
    Scale Alpha=&FloatFromCurve(FloatCurve=DataInterface() {propertyValues {Curve {keys [<at 0 1>,<at 0.6 0.6>,<at 1 0>]}}})
    Scale RGB=&MakeVectorFromLinearColorRGB(Color=&ColorFromCurve(ColorCurve=DataInterface() {propertyValues {
        AlphaCurve {keys []}
        BlueCurve {keys [<at 0.0734 0.7691 RCIM_Cubic>,<at 0.1716 1>,<at 0.4113 0.9488 RCIM_Cubic>]}
        GreenCurve {keys [<at 0.0734 0.6421 RCIM_Cubic>,<at 0.1716 0.3893>,<at 0.4113 0.9392 RCIM_Cubic>]}
        RedCurve {keys [<at 0.0734 1 RCIM_Cubic>,<at 0.1716 0.666>,<at 0.4113 1 RCIM_Cubic>]}
       }}))
    ScaleA=true
    ScaleRGB=true
   )
   &ScaleColor(
    Scale Alpha=&FloatFromCurve(FloatCurve=DataInterface() {propertyValues {Curve {keys [<at 0 0.5>,<at 0.8 0>]}}}, CurveIndex=$System.Age)
    ScaleA=true
    ScaleRGB=true
   ) {module=ScaleColor001}
  ]
  rendererClasses [&NiagaraRibbonRendererProperties]
  renderers [{rendererClass=&NiagaraRibbonRendererProperties}]
 }
 Dirt_TrailHead {
  emitterUpdateScript [
   &EmitterState(Inactive Response=::NewEnumerator1, Life Cycle Mode=::NewEnumerator1, Loop Behavior=::NewEnumerator1, Loop Duration=1, UseLoopDelay=false)
   &SpawnBurst_Instantaneous(Spawn Count=40, Use Spawn Probability=false, UseLoopCountLimit=false)
   &SpawnRate(SpawnRate=10, Use Spawn Probability=false) {enabled=false}
  ]
  particleSpawnScript [
   &InitializeParticle(
    Lifetime Max=0.8
    Lifetime Min=0.35
    Lifetime Mode=::NewEnumerator1
    Position Offset=$System.NS Position
    Sprite Size Mode=::NewEnumerator1
    Uniform Sprite Size Max=0.5
    Uniform Sprite Size Min=0.3
    UsePositionOffset=true
    Write Material Random=true
   )
   &ShapeLocation(
    MaskSpawnGroup=false
    Override Seed=false
    Rotation Angle=0.2
    Rotation Axis=<0 1 0>
    Rotation Mode=::NewEnumerator1
    Sphere Radius=5
    UseOwnerScale=true
   )
   &AddVelocity(
    Module.Use Falloff Distance=false
    Origin Offset=$System.NS Position
    Override Seed=false
    Velocity Mode=::NewEnumerator1
    Velocity Speed=&RandomRangeFloat(Minimum=100, Maximum=1000)
   )
   &AddVelocity(Velocity=<0 0 1>) {module=AddVelocity001}
  ]
  particleUpdateScript [
   &ParticleState()
   &SpriteRotationRate(Rotation Rate=&Multiply_Float(A=&RandomRangeFloat(Minimum=-1, Maximum=1), B=800))
   &Drag(Drag=&RandomRangeFloat(Minimum=3, Maximum=5), Use Linear Drag=true, Use Rotational Drag=true)
   &GravityForce(Gravity=&RandomRangeVector(Minimum=<0 0 -300>, Maximum=<0 0 -600>))
   &SolveForcesAndVelocity(Clamp Velocity=false, Limit Acceleration=false)
   &ScaleSpriteSizeBySpeed(Max Scale Factor=<0.5 8>, Min Scale Factor=<0.4 5>)
   &GenerateLocationEvent(Use Event Probability=false, UseEventDelay=false)
   &ScaleColor(
    Scale RGB=/Niagara/DynamicInputs/Math/LerpMultipleVectors(
     Alpha:NiagaraFloat=$User.VariantSwitch
     V 1:Vector3f=&MakeVectorFromLinearColorRGB(Color=&ColorFromCurve(ColorCurve=DataInterface() {propertyValues {
         AlphaCurve {keys []}
         BlueCurve {keys [<at 0 1.9927 RCIM_Cubic>,<at 0.132 1>,<at 0.52 0.25 RCIM_Cubic>]}
         GreenCurve {keys [<at 0 0.9332 RCIM_Cubic>,<at 0.132 0.5153>,<at 0.52 0.063 RCIM_Cubic>]}
         RedCurve {keys [<at 0 2 RCIM_Cubic>,<at 0.132 0.9338>,<at 0.52 0.2489 RCIM_Cubic>]}
        }}))
     V 2:Vector3f=&MakeVectorFromLinearColorRGB(Color=&ColorFromCurve(ColorCurve=DataInterface() {propertyValues {
         AlphaCurve {keys []}
         BlueCurve {keys [<at 0 0.2092 RCIM_Cubic>,<at 0.132 0.3519>,<at 0.52 0.0824 RCIM_Cubic>]}
         GreenCurve {keys [<at 0 0.2552 RCIM_Cubic>,<at 0.132 1>,<at 0.52 0.1406 RCIM_Cubic>]}
         RedCurve {keys [<at 0 0.0086 RCIM_Cubic>,<at 0.132 0>,<at 0.52 0.0161 RCIM_Cubic>]}
        }}))
     V 3:Vector3f=?
     V 4:Vector3f=?
     V 5:Vector3f=?
     Vector Count:ENiagara_MultipleLerpCount=::NewEnumerator3
    )
    ScaleA=true
    ScaleRGB=true
   )
  ]
  rendererClasses [&NiagaraSpriteRendererProperties]
  renderers [{rendererClass=&NiagaraSpriteRendererProperties}]
 }
 Flash_Flare {
  emitterUpdateScript [
   &EmitterState(Life Cycle Mode=::NewEnumerator1, Loop Behavior=::NewEnumerator1, Loop Duration=1, UseLoopDelay=false)
   &SpawnBurst_Instantaneous(Spawn Count=1, Use Spawn Probability=false)
  ]
  particleSpawnScript [
   &InitializeParticle(
    Color=<rgba 0.8646 0.2591 0.6957 0.5>
    Lifetime=0.15
    Sprite Size Mode=::NewEnumerator3
    Uniform Sprite Size=&FN_FrontEndFloatModifier(Gameplay=350, Front End=250)
    UsePositionOffset=false
    Write Material Random=true
   )
   &DynamicMaterialParameters(
    Index 0 Param 1=0.2
    Index 0 Param 2=1
    Module.Param0WriteEnabled=true
    Module.Param1WriteEnabled=true
    Module.Param2WriteEnabled=true
    Module.Param3WriteEnabled=true
   )
  ]
  particleUpdateScript [
   &ParticleState()
   &ScaleSpriteSize(Uniform Curve Sprite Scale=DataInterface() {propertyValues {Curve {keys [{tangentMode=RCTM_User, time=0, value=0.6}, {tangentMode=RCTM_User, time=0.5, value=1}]}}})
   &CameraOffset(Camera Offset Amount=45)
   &SolveForcesAndVelocity(Clamp Velocity=false, Limit Acceleration=false)
   &ScaleColor(
    Scale Alpha=&Multiply_Float(A=&Lerp_Float(A=1, B=0, Alpha=$Particles.NormalizedAge), B=0.3)
    Scale RGB=&VectorFromFloat(Value=&FNTimeOfDayScalar(Night Scale=0.5, Morning Scale=1, Evening Scale=1))
    ScaleA=true
    ScaleRGB=true
   )
   &FN_Light_Attributes(Create Exponent Attribute=true, Create Position Attribute=true, Create Radius Attribute=true, Exponent=2, Radius=150)
   /SetVariables(Particles.Light Color:LinearColor=&ColorFromCurve(ColorCurve=DataInterface() {propertyValues {
       AlphaCurve {keys [<at 0 1 RCIM_Cubic>,<at 1 0 RCIM_Cubic>]}
       BlueCurve {keys [<at 0 0>,<at 0.384 5 RCIM_Cubic>,<at 0.9971 0>]}
       GreenCurve {keys [<at 0 0>,<at 0.384 1.5 RCIM_Cubic>,<at 0.9971 0>]}
       RedCurve {keys [<at 0 0>,<at 0.384 3.8 RCIM_Cubic>,<at 0.9971 0>]}
      }})) {isSetParameters=true}
  ]
  rendererClasses [&NiagaraSpriteRendererProperties, &NiagaraLightRendererProperties]
  renderers [
   {rendererClass=&NiagaraSpriteRendererProperties}
   {rendererClass=&NiagaraLightRendererProperties, rendererIndex=1}
  ]
 }
 IceChunks {
  emitterUpdateScript [
   &EmitterState(Life Cycle Mode=::NewEnumerator1, Loop Behavior=::NewEnumerator1, Loop Duration=2, UseLoopDelay=false)
   &SpawnBurst_Instantaneous(Spawn Count=180, Use Spawn Probability=false, UseLoopCountLimit=false)
  ]
  particleSpawnScript [
   &InitializeParticle(
    Color=<rgba 7.5 1.77 3.1452 1>
    Lifetime Max=1
    Lifetime Min=0.25
    Lifetime Mode=::NewEnumerator1
    Mesh Renderer Array Visibility Mode=::NewEnumerator1
    Mesh Renderer Info=DataInterface() {propertyValues {MeshRenderer=&self:IceChunks_0.NiagaraMeshRendererProperties_0}}
    Mesh Renderer Mesh Index:NiagaraInt32=&RandomRangeInt(Minimum=0, Maximum=2)
    Mesh Scale Mode=::NewEnumerator3
    Mesh Uniform Scale=&RandomRangeFloat(Minimum=0.1, Maximum=0.15)
    UsePositionOffset=false
    Write Material Random=true
   )
   &ShapeLocation(MaskSpawnGroup=false, Override Seed=false, Sphere Radius=11, UseOwnerScale=true)
   &InitialMeshOrientation(
    Mesh Axis To Orient=<0 0 1>
    Mesh Orientation Mode=::NewEnumerator2
    Orientation Vector=$Particles.ShapeLocation.ShapeNormal
    Use Rotation Vector=false
   )
   &AddVelocity(Module.Use Falloff Distance=false, Override Seed=false, Velocity Mode=::NewEnumerator1, Velocity Speed=400)
   &AddVelocity(Rotation Coordinate Space=::World, Velocity=&RandomRangeVector(Minimum=<0 0 50>, Maximum=<0 0 400>)) {module=AddVelocity001}
  ]
  particleUpdateScript [
   &ParticleState()
   &UpdateMeshOrientation(
    Rotation Rate=&RandomRangeFloat(Minimum=2, Maximum=4)
    Rotation Vector=&RandomVector(Vector Scale=1)
   )
   &ScaleMeshSize(Scale Factor=&VectorFromFloat(Value=&FloatFromCurve(FloatCurve=DataInterface() {propertyValues {Curve {keys [
          {interpMode=RCIM_Cubic, tangentMode=RCTM_User, time=0, value=1}
          {arriveTangent=-1.6697, interpMode=RCIM_Cubic, leaveTangent=-1.6697, tangentMode=RCTM_User, time=1, value=0}
         ]}}})))
   &Drag(Drag=&RandomRangeFloat(Minimum=2, Maximum=3), Use Linear Drag=true, Use Rotational Drag=true)
   &CurlNoiseForce(Noise Strength=111, Pan Noise=true, Pan Noise Field=<0 0 0.1>)
   &GravityForce()
   &SolveForcesAndVelocity(Clamp Velocity=false, Limit Acceleration=false)
  ]
  rendererClasses [&NiagaraMeshRendererProperties]
  renderers [{rendererClass=&NiagaraMeshRendererProperties, rendererIndex=0}]
 }
 Smoke {
  emitterUpdateScript [
   &EmitterState()
   &SpawnBurst_Instantaneous(Spawn Count=32, Use Spawn Probability=false, UseLoopCountLimit=false)
   &SpawnBurst_Instantaneous(Spawn Count=7, Spawn Group=1, Use Spawn Probability=false, UseLoopCountLimit=false) {module=SpawnBurst_Instantaneous001}
   &SpawnBurst_Instantaneous(Spawn Count=25, Spawn Group=2, Use Spawn Probability=false, UseLoopCountLimit=false) {module=SpawnBurst_Instantaneous002}
   &SpawnBurst_Instantaneous(Spawn Count=14, Spawn Group=3, Use Spawn Probability=false) {module=SpawnBurst_Instantaneous003}
   &SpawnBurst_Instantaneous(Spawn Group=4, Use Spawn Probability=false, UseLoopCountLimit=false) {module=SpawnBurst_Instantaneous004}
   /SetVariables(Emitter.Linear Color:LinearColor=&FNTimeOfDayColor(
     Night Color=<rgba 1.5943 0.4975 2.5 1>
     Morning Color=<rgba 1 0.2453 0.6479 1>
     Day Color=<rgba 2 0.5761 1.2749 1>
     Evening Color=<rgba 1 0.5229 0.7855 1>
    )) {isSetParameters=true}
  ]
  particleSpawnScript [
   /SetVariables(Particles.VisibilityTag=&SpawnGroup()) {isSetParameters=true}
   &InitializeParticle(
    Color=&FNTimeOfDayColor(
     Night Color=<rgba 0.705 0.41 1 1>
     Morning Color=<rgba 0.4642 0.3999 1 1>
     Day Color=<rgba 2 0.4799 1.933 0.4>
     Evening Color=<rgba 0.7663 1.5907 2 0.4>
    )
    Lifetime=&MaskFloatBySpawnGroup(
     Masked Float=&RandomRangeFloat(Minimum=1, Maximum=2.3)
     Passthrough Float=&MaskFloat(
      Mask Source=0
      Mask Parameter=?
      Mask=3
      Masked Float=&UniformRangedFloat(Minimum=1, Maximum=1.7)
      Passthrough Float=&MaskFloatBySpawnGroup(Masked Float=0.4, Passthrough Float=&RandomRangeFloat(Minimum=0.3, Maximum=1.2), Spawn Group Mask=4)
     )
     Spawn Group Mask=2
    )
    Mesh Scale Mode=::NewEnumerator3
    Mesh Uniform Scale=1.1
    Sprite Size=&MaskVector2dBySpawnGroup(Masked Vector2D=<20 30>, Passthrough Vector2D=<60 30>, Spawn Group Mask=0)
    Sprite Size Mode=::NewEnumerator4
    UsePositionOffset=false
    Write Material Random=true
   )
   &ApplyInitialForces()
   &AddVelocityFromPoint(
    Module.Use Falloff Distance=false
    Origin Offset=&RandomRangeVector(Minimum=<0 0 0>, Maximum=<0 0 0>)
    Velocity Strength=&RandomRangeFloat(
     Minimum=&MaskFloatBySpawnGroup(
      Masked Float=99
      Passthrough Float=&MaskFloatBySpawnGroup(
       Masked Float=0
       Passthrough Float=&MaskFloatBySpawnGroup(Masked Float=0, Passthrough Float=150, Spawn Group Mask=4)
       Spawn Group Mask=3
      )
      Spawn Group Mask=2
     )
     Maximum=&MaskFloatBySpawnGroup(
      Masked Float=150
      Passthrough Float=&MaskFloatBySpawnGroup(
       Masked Float=0
       Passthrough Float=&MaskFloatBySpawnGroup(Masked Float=0, Passthrough Float=400, Spawn Group Mask=4)
       Spawn Group Mask=3
      )
      Spawn Group Mask=2
     )
    )
   )
   &AddVelocityInCone(
    Cone Angle=80.4
    Cone Axis=<0 0 1>
    Cone Axis Coordinate Space=::World
    Module.Use Velocity Falloff On Cone Axis=false
    Velocity Strength=&MaskFloat(
     Mask Source=0
     Mask Parameter=?
     Mask=3
     Masked Float=&UniformRangedFloat(Minimum=150, Maximum=300)
     Passthrough Float=0
    )
   )
   &SphereLocation(
    MaskSpawnGroup=false
    Offset=&MaskVectorBySpawnGroup(
     Masked Vector=<0 0 -20>
     Passthrough Vector=&MaskVectorBySpawnGroup(
      Masked Vector=<0 0 0>
      Passthrough Vector=&MaskVectorBySpawnGroup(Masked Vector=<0 0 0>, Passthrough Vector=<1 1 -10>, Spawn Group Mask=4)
      Spawn Group Mask=3
     )
     Spawn Group Mask=2
    )
    Offset Coordinate Space:ENiagaraCoordinateSpace=::World
    Override Seed=true
    Random Seed=20
    Sphere Radius=&MaskFloatBySpawnGroup(
     Masked Float=0
     Passthrough Float=&MaskFloatBySpawnGroup(Masked Float=0, Passthrough Float=5, Spawn Group Mask=4)
     Spawn Group Mask=3
    )
    Surface Only=false
    UseNonUniformScale=false
    UseOffset=true
    UseSphereOrientation=false
   )
   &InitialMeshOrientation(Mesh Axis To Orient=<0 0 1>, Mesh Orientation Mode=::NewEnumerator2, Orientation Vector=<0 0 1>, Use Rotation Vector=false)
   &CameraOffset(Camera Offset Amount=&MaskFloatBySpawnGroup(Masked Float=5, Passthrough Float=0, Spawn Group Mask=4))
   /SetVariables(Particles.Mist_Rotation:NiagaraFloat=&RandomRangeFloat(Minimum=-360, Maximum=360)) {isSetParameters=true}
  ]
  particleUpdateScript [
   &ParticleState()
   &ScaleMeshSize(Scale Factor=&VectorFromFloat(Value=&FloatFromCurve(FloatCurve=DataInterface() {propertyValues {Curve {keys [
          {arriveTangent=3.1045, interpMode=RCIM_Cubic, leaveTangent=3.1045, leaveTangentWeight=0.8949, tangentMode=RCTM_User, time=0, value=0}
          {arriveTangentWeight=0.4784, interpMode=RCIM_Cubic, tangentMode=RCTM_User, time=1, value=1}
         ]}}})))
   /SetVariables(
    Particles.MainCloud_Color:LinearColor=&MultiplyLinearColor(
     Linear Color=&FNTimeOfDayColor(
      Night Color=<rgba 1 0.4314 0.5176 0.502>
      Morning Color=<rgba 0.4378 0.5174 1 1>
      Day Color=<rgba 1.1027 0.5044 1.5 0.5>
      Evening Color=<rgba 1 0.4314 0.5176 0.502>
     )
     Scale Factor=&MakeVector4FromVectorAndFloat(
      "Vector (RGB)"=&VectorFromFloat(Value=&FNTimeOfDayScalar(
        Night Scale=0.8
        Morning Scale=0.8
        Day Scale=&FN_FrontEndFloatModifier(Gameplay=1, Front End=1)
        Evening Scale=0.5
       ))
      "Float (Alpha)"=&FloatFromCurve(
       FloatCurve=DataInterface() {propertyValues {Curve {keys [
           {arriveTangent=-0.0888, interpMode=RCIM_Cubic, leaveTangent=-0.0888, tangentMode=RCTM_User, time=0, value=0}
           {arriveTangent=0.12, interpMode=RCIM_Cubic, leaveTangent=0.12, tangentMode=RCTM_User, time=0.15, value=0.9847}
           {arriveTangent=-0.0405, interpMode=RCIM_Cubic, leaveTangent=-0.0405, tangentMode=RCTM_User, time=1, value=0}
          ]}}}
       Scale Curve=&FNTimeOfDayScalar(
        Night Scale=1
        Morning Scale=1
        Day Scale=&FN_FrontEndFloatModifier(Gameplay=1, Front End=1)
        Evening Scale=1
       )
      )
     )
    )
    Particles.MainCloud_Size:Vector2f=&Multiply_Vector2D_ByFloat(
     Float=&FloatFromCurve(FloatCurve=DataInterface() {propertyValues {Curve {keys [<at 0.001 0.5>,<at 0.1 1.1652>,<at 1.0009 0.8072>]}}})
     Vector2D=&RandomRangeVector2D(Minimum=<80 80>, Maximum=<140 140>)
    )
   ) {isSetParameters=true}
   /SetVariables(
    Particles.MistColor:LinearColor=&MultiplyLinearColor(
     Linear Color=<rgba 1 0.3003 0.9888 0.3>
     Scale Factor=&MakeVector4FromVectorAndFloat(
      "Vector (RGB)"=<1 1 1>
      "Float (Alpha)"=&FloatFromCurve(FloatCurve=DataInterface() {propertyValues {Curve {keys [<at 0 0 RCIM_Cubic>, {arriveTangent=0.002, interpMode=RCIM_Cubic, leaveTangent=0.002, time=0.1942, value=2}, <at 1.0011 0.002 RCIM_Cubic>]}}})
     )
    )
    Particles.MistSize:Vector2f=&Multiply_Vector2D_ByFloat(
     Float=&FloatFromCurve(FloatCurve=DataInterface() {propertyValues {Curve {keys [<at -0.0046 0.5>,<at 1.0011 2>]}}}, Scale Curve=1.5)
     Vector2D=&RandomRangeVector2D(Minimum=<50 50>, Maximum=<100 100>)
    )
   ) {isSetParameters=true}
   /SetVariables(Particles.WP_Color:LinearColor=&MultiplyLinearColor(
     Linear Color=<rgba 0 0.6493 1 0.4365>
     Scale Factor=&MakeVector4FromVectorAndFloat(
      "Vector (RGB)"=<1 1 1>
      "Float (Alpha)"=&FloatFromCurve(
       FloatCurve=DataInterface() {propertyValues {Curve {keys [
           {interpMode=RCIM_Cubic, leaveTangentWeight=2.5689, tangentMode=RCTM_User, time=0, value=0}
           {interpMode=RCIM_Cubic, tangentMode=RCTM_User, time=0.1962, value=1}
           {interpMode=RCIM_Cubic, tangentMode=RCTM_User, time=1, value=0}
          ]}}}
       Scale Curve=&FNTimeOfDayScalar(Night Scale=1, Morning Scale=1, Evening Scale=1)
      )
     )
    )) {isSetParameters=true}
   &SubUVAnimation(
    Output Alternate Renderer Binding=true
    "Sprite Renderer "=DataInterface() {propertyValues {SpriteRenderer=&self:Smoke_0.NiagaraSpriteRendererProperties_4}}
    UseEndFrame=false
    UseStartFrame=false
   )
   &ScaleSpriteSize(
    Scale Sprite Size Mode=::NewEnumerator0
    Uniform Scale Factor=&FloatFromCurve(FloatCurve=DataInterface() {propertyValues {Curve {keys [<at 0.001 0.5>,<at 0.2948 1.1652>,<at 1.0009 0.8072>]}}})
   )
   &ScaleColor(
    Linear Color Curve=DataInterface() {propertyValues {
      AlphaCurve {keys [<at 0.0057 1 RCIM_Cubic>]}
      BlueCurve {keys [<at 0 0.6519>,<at 0.666 0.7333>]}
      GreenCurve {keys [<at 0 0.414>,<at 0.666 0.6>]}
      RedCurve {keys [<at 0 2>,<at 0.666 1>]}
     }}
    Scale Mode=::NewEnumerator2
   ) {module=ScaleColor001}
   &ScaleColor(
    Scale Alpha=&FloatFromCurve(
     FloatCurve=DataInterface() {propertyValues {Curve {keys [
         {arriveTangent=-0.0888, interpMode=RCIM_Cubic, leaveTangent=-0.0888, tangentMode=RCTM_User, time=0, value=0}
         {arriveTangent=0.12, interpMode=RCIM_Cubic, leaveTangent=0.12, tangentMode=RCTM_User, time=0.15, value=0.9847}
         {arriveTangent=-0.0405, interpMode=RCIM_Cubic, leaveTangent=-0.0405, tangentMode=RCTM_User, time=0.6243, value=0}
        ]}}}
     Scale Curve=&FNTimeOfDayScalar(
      Night Scale=1
      Morning Scale=1
      Day Scale=&FN_FrontEndFloatModifier(Gameplay=1, Front End=1)
      Evening Scale=1
     )
    )
    Scale RGB=&VectorFromFloat(Value=&FNTimeOfDayScalar(
      Night Scale=0.8
      Morning Scale=0.8
      Day Scale=&FN_FrontEndFloatModifier(Gameplay=1, Front End=1)
      Evening Scale=0.5
     ))
    ScaleA=true
    ScaleRGB=true
   ) {module=ScaleColor003}
   &Drag(
    Drag=&MaskFloatBySpawnGroup(
     Masked Float=&FloatFromCurve(
      FloatCurve=DataInterface() {propertyValues {Curve {keys [
          {arriveTangent=0.7001, interpMode=RCIM_Cubic, leaveTangent=0.7001, tangentMode=RCTM_User, time=0, value=0}
          {interpMode=RCIM_Cubic, tangentMode=RCTM_User, time=1, value=1}
         ]}}}
      Scale Curve=4
     )
     Passthrough Float=&MaskFloatBySpawnGroup(Masked Float=1, Passthrough Float=3, Spawn Group Mask=3)
     Spawn Group Mask=2
    )
    Use Linear Drag=true
    Use Rotational Drag=true
   ) {module=Drag002}
   &GravityForce(Gravity=&MaskVectorBySpawnGroup(Masked Vector=&Multiply_Vector(A=<0 0 -980>, B=<0 0 0.5>), Passthrough Vector=<0 0 0>, Spawn Group Mask=3))
   &SolveForcesAndVelocity(Clamp Velocity=false, Limit Acceleration=false)
   &ScaleSpriteSizeBySpeed(
    Max Scale Factor=&MaskVector2dBySpawnGroup(Masked Vector2D=<1 1>, Passthrough Vector2D=<1 5>, Spawn Group Mask=2)
    Velocity Threshold=100
   )
   &DynamicMaterialParameters(
    Index 0 Param 1=&Add_Float(
     A=&RandomRangeFloat(Minimum=0, Maximum=1)
     B=&FloatFromCurve(
      FloatCurve=DataInterface() {propertyValues {Curve {keys [
          {arriveTangent=2.7533, interpMode=RCIM_Cubic, leaveTangent=1.8958, tangentMode=RCTM_User, time=0, value=1}
          {arriveTangent=0.8292, interpMode=RCIM_Cubic, leaveTangent=0.8292, tangentMode=RCTM_User, time=0.321, value=1.3453}
          {arriveTangent=0.0529, interpMode=RCIM_Cubic, leaveTangent=0.2021, tangentMode=RCTM_User, time=1, value=1.5}
         ]}}}
      CurveIndex=&Add_Float(A=&Multiply_Float(A=1, B=0.5), B=$Particles.NormalizedAge)
     )
    )
    Index 0 Param 2=&FloatFromCurve(FloatCurve=DataInterface() {propertyValues {Curve {keys [
         {arriveTangent=0.0453, interpMode=RCIM_Cubic, tangentMode=RCTM_Break, time=0, value=-1}
         {arriveTangent=0.0604, interpMode=RCIM_Cubic, leaveTangent=0.0604, tangentMode=RCTM_User, time=1, value=-1}
        ]}}})
    Index 0 Param 3=&FN_Select_Float_from_Time_Of_Day(Morning=2.5, Day=0, Evening=2.5, Night=5)
    Index 0 Param 4=&FloatFromCurve(
     FloatCurve=DataInterface() {propertyValues {Curve {keys [<at 0.5 1>,<at 1 0>]}}}
     Scale Curve=&FNTimeOfDayScalar(
      Night Scale=2
      Morning Scale=2
      Day Scale=&FN_FrontEndFloatModifier(Gameplay=1, Front End=0.5)
      Evening Scale=3
     )
    )
    Index 2 Param 1=&FloatFromCurve(FloatCurve=DataInterface() {propertyValues {Curve {keys [
         {interpMode=RCIM_Cubic, leaveTangent=4.2485, tangentMode=RCTM_Break, time=0, value=0}
         {arriveTangent=1.0548, interpMode=RCIM_Cubic, leaveTangent=1.0548, tangentMode=RCTM_User, time=0.17, value=0.5715}
         <at 1 1>
        ]}}})
    Index 2 Param 2=&FloatFromCurve(
     FloatCurve=DataInterface() {propertyValues {Curve {keys [
         {arriveTangent=1.9056, interpMode=RCIM_Cubic, leaveTangent=1.9056, tangentMode=RCTM_User, time=0, value=0.2}
         {arriveTangent=0.0259, interpMode=RCIM_Cubic, leaveTangent=0.0259, tangentMode=RCTM_User, time=1, value=1.4}
        ]}}}
     Scale Curve=&RandomRangeFloat(Minimum=0.2, Maximum=0.5)
    )
    Index 2 Param 3=&FloatFromCurve(FloatCurve=DataInterface() {propertyValues {Curve {keys [
         {interpMode=RCIM_Cubic, leaveTangent=2.3064, tangentMode=RCTM_Break, time=0, value=1}
         {interpMode=RCIM_Cubic, leaveTangent=11.1233, tangentMode=RCTM_Break, time=1, value=2}
        ]}}})
    Index 2 Param 4=&FloatFromCurve(
     FloatCurve=DataInterface() {propertyValues {Curve {keys [
         {arriveTangent=-0.6867, interpMode=RCIM_Cubic, leaveTangent=-0.6867, tangentMode=RCTM_User, time=0, value=1}
         {interpMode=RCIM_Cubic, leaveTangent=-1.2956, tangentMode=RCTM_Break, time=1, value=0.75}
        ]}}}
     Scale Curve=1.6
    )
    Module.Param0WriteEnabled=true
    Module.Param0WriteEnabled2=true
    Module.Param1WriteEnabled=false
    Module.Param1WriteEnabled2=true
    Module.Param2WriteEnabled=true
    Module.Param2WriteEnabled2=true
    Module.Param3WriteEnabled=true
    Module.Param3WriteEnabled2=true
    Write Parameter Index 2=true
   )
   /SetVariables(
    Particles.MainCloud_DMP:Vector4f=&MakeVector4(
     X=&Add_Float(
      A=&RandomRangeFloat(Minimum=0, Maximum=1)
      B=&FloatFromCurve(
       FloatCurve=DataInterface() {propertyValues {Curve {keys [
           {arriveTangent=2.7533, interpMode=RCIM_Cubic, leaveTangent=1.8958, tangentMode=RCTM_User, time=0, value=1}
           {arriveTangent=0.8292, interpMode=RCIM_Cubic, leaveTangent=0.8292, tangentMode=RCTM_User, time=0.321, value=1.3453}
           {arriveTangent=0.0529, interpMode=RCIM_Cubic, leaveTangent=0.2021, tangentMode=RCTM_User, time=1, value=1.5}
          ]}}}
       CurveIndex=&Add_Float(A=&Multiply_Float(A=1, B=0.5), B=$Particles.NormalizedAge)
      )
     )
     Y=0
     Z=0.5
     W=&FloatFromCurve(
      FloatCurve=DataInterface() {propertyValues {Curve {keys [<at 0.5 1>,<at 1 0>]}}}
      Scale Curve=&FNTimeOfDayScalar(
       Night Scale=2
       Morning Scale=2
       Day Scale=&FN_FrontEndFloatModifier(Gameplay=1, Front End=0.5)
       Evening Scale=3
      )
     )
    )
    Particles.MainCloud_DMP2:Vector4f=&MakeVector4(
     X=&FloatFromCurve(FloatCurve=DataInterface() {propertyValues {Curve {keys [
          {interpMode=RCIM_Cubic, leaveTangent=4.2485, tangentMode=RCTM_Break, time=0, value=0}
          {arriveTangent=1.0548, interpMode=RCIM_Cubic, leaveTangent=1.0548, tangentMode=RCTM_User, time=0.17, value=0.5715}
          <at 1 1>
         ]}}})
     Y=&FloatFromCurve(
      FloatCurve=DataInterface() {propertyValues {Curve {keys [
          {arriveTangent=1.9056, interpMode=RCIM_Cubic, leaveTangent=1.9056, tangentMode=RCTM_User, time=0, value=0.2}
          {arriveTangent=0.0259, interpMode=RCIM_Cubic, leaveTangent=0.0259, tangentMode=RCTM_User, time=1, value=1.4}
         ]}}}
      Scale Curve=&RandomRangeFloat(Minimum=0.2, Maximum=0.5)
     )
     Z=&FloatFromCurve(FloatCurve=DataInterface() {propertyValues {Curve {keys [
          {interpMode=RCIM_Cubic, leaveTangent=2.3064, tangentMode=RCTM_Break, time=0, value=1}
          {interpMode=RCIM_Cubic, leaveTangent=11.1233, tangentMode=RCTM_Break, time=1, value=2}
         ]}}})
     W=&FloatFromCurve(
      FloatCurve=DataInterface() {propertyValues {Curve {keys [
          {arriveTangent=-0.6867, interpMode=RCIM_Cubic, leaveTangent=-0.6867, tangentMode=RCTM_User, time=0, value=1}
          {interpMode=RCIM_Cubic, leaveTangent=-1.2956, tangentMode=RCTM_Break, time=1, value=0.75}
         ]}}}
      Scale Curve=1.6
     )
    )
   ) {isSetParameters=true}
   /SetVariables(Particles.Mist_DMP:Vector4f=&MakeVector4(X=&UniformRangedFloat(Minimum=4, Maximum=7), Y=8, Z=1, W=0.01)) {isSetParameters=true}
  ]
  rendererClasses [&NiagaraSpriteRendererProperties, &NiagaraMeshRendererProperties]
  renderers [
   {rendererClass=&NiagaraSpriteRendererProperties}
   {rendererClass=&NiagaraSpriteRendererProperties, rendererIndex=1}
   {rendererClass=&NiagaraSpriteRendererProperties, rendererIndex=2}
   {rendererClass=&NiagaraSpriteRendererProperties, rendererIndex=3}
   {rendererClass=&NiagaraMeshRendererProperties, rendererIndex=4}
  ]
 }
}
summary {
 systemName=NS_Emote_TealMink_BagPoof_Compressed
 emitters [
  {emitterName=Coin, rendererClasses [&NiagaraMeshRendererProperties, &NiagaraSpriteRendererProperties]}
  {emitterName=Smoke, rendererClasses [&NiagaraSpriteRendererProperties, &NiagaraMeshRendererProperties]}
  {emitterName=IceChunks, rendererClasses [&NiagaraMeshRendererProperties]}
  {emitterName=Flash_Flare, rendererClasses [&NiagaraSpriteRendererProperties, &NiagaraLightRendererProperties]}
  {emitterName=Dirt_Trail, rendererClasses [&NiagaraRibbonRendererProperties]}
  {emitterName=Dirt_TrailHead, rendererClasses [&NiagaraSpriteRendererProperties]}
  {emitterName=BlowingParticles, rendererClasses [&NiagaraDecalRendererProperties]}
 ]
 userVariables [
  User.bIsFrontEnd {type=NiagaraBool}
  User.bIsFrontEndPreview {type=NiagaraBool}
  User.bIsPainted {type=NiagaraBool}
  User.HourGlassSize {type=NiagaraFloat, defaultValue=0, description=""}
  User.PaintedColor {type=LinearColor, description="", defaultValue {struct=LinearColor, value=<rgba 1 1 1 1>}}
  User.ParaGlideLeanAlpha {type=NiagaraFloat, defaultValue=0, description=""}
  User.RibbonIDOffset {type=NiagaraInt32}
  User.SkeletalMesh {
   type=NiagaraDataInterfaceSkeletalMesh
   defaultValue {
    struct=NiagaraExt_VariableValue_DataInterface
    value {dataInterface=&self:NiagaraDataInterfaceSkeletalMesh_0}
   }
  }
  User.VariantSwitch {type=NiagaraFloat, defaultValue=0, description=""}
 ]
}
