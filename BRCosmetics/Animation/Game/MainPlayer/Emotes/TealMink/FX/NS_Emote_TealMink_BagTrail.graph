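# Niagara system named as a bag trail effect for the TealMink emote, though it contains no emitters as defined here.
# Its system scripts use only System State and a spawn-time Set Variables assignment.
# It exposes two skeletal mesh data interfaces, front-end and preview flags, a paraglide lean alpha and a ribbon ID offset as user parameters, with no emitter to render from them.
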NiagaraSystem /BRCosmetics/Animation/Game/MainPlayer/Emotes/TealMink/FX/NS_Emote_TealMink_BagTrail
dependencies {
 usedDataInterfaces []
 usedDynamicInputs []
 usedModules [
  &SystemState
  &self:SystemSpawnScript.SystemScriptSource.SystemScriptGraph.NiagaraNodeAssignment_0.SetVariables_#1
 ]
}
summary {
 systemName=NS_Emote_TealMink_BagTrail
 emitters []
 userVariables [
  User.bIsFrontEnd {type=NiagaraBool}
  User.bIsFrontEndPreview {type=NiagaraBool}
  User.ParaGlideLeanAlpha {type=NiagaraFloat, defaultValue=0, description=""}
  User.RibbonIDOffset {type=NiagaraInt32}
  User.Skeletal Mesh02 {
   type=NiagaraDataInterfaceSkeletalMesh
   defaultValue {
    struct=NiagaraExt_VariableValue_DataInterface
    value {dataInterface=&self:NiagaraDataInterfaceSkeletalMesh_1}
   }
  }
  User.SkeletalMesh {
   type=NiagaraDataInterfaceSkeletalMesh
   defaultValue {
    struct=NiagaraExt_VariableValue_DataInterface
    value {dataInterface=&self:NiagaraDataInterfaceSkeletalMesh_0}
   }
  }
 ]
}
